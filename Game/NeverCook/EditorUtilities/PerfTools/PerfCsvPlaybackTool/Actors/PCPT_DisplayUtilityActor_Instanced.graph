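# Editor utility actor for the Perf CSV Playback Tool that displays data points and view frustums in the world as instanced static meshes.
# addPoint adds a point and a frustum instance, storing index, color and value in per-instance custom data.
# Setters write current index, scale, draw-every-other and value threshold to custom primitive data, presumably for materials to scale, highlight and filter instances.
# validateFrustumVisual applies frustum visibility, forced-on, opacity and scale times distance, on BeginPlay and frustum setting changes.
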EditorUtilityBlueprint /Game/NeverCook/EditorUtilities/PerfTools/PerfCsvPlaybackTool/Actors/PCPT_DisplayUtilityActor_Instanced
graphs [
 EventGraph {nodes [
   Event BeginPlay()
   > validateFrustumVisual()
  ]}
 Set Current Index {nodes [
   n1 = Set Current Index!() -> (then, index:int) {is=title K2Node_FunctionEntry}
   > Set currentIndex(currentIndex:int=@n1) -> (Output_Get:int=0)
   > SetCustomPrimitiveDataFloat!(
    execute
    self:Primitive Component Object Reference=[
     Get FrustumInstancedMesh() -> (FrustumInstancedMesh:Instanced Static Mesh Component Object Reference)
     Get PointsInstancedMesh() -> (PointsInstancedMesh:Instanced Static Mesh Component Object Reference)
    ]
    DataIndex:int=1
    Value:float="To Float (Integer)"(InInt=Get currentIndex() -> (currentIndex:int=0))
   ) -> (then)
  ]}
 UserConstructionScript {nodes [Construction Script()]}
 addPoint {nodes [
   n1 = addPoint!() -> (then, color:Linear Color Structure, WorldTransform:"Transform (by ref)", pointValue:double=~) {is=title K2Node_FunctionEntry}
   > Set l_color(l_color:Linear Color Structure=@n1) -> (Output_Get:Linear Color Structure)
   n2 = BreakTransform(InTransform=@n1.WorldTransform)
   > Set l_transform(l_transform:Transform=MakeTransform(Location=@n2, Rotation=@n2.Rotation, Scale=@n2.Scale)) -> (Output_Get:Transform)
   > n3 = Sequence()
   > n4 = AddInstance!(
    execute
    self:Instanced Static Mesh Component Object Reference=Get PointsInstancedMesh()
    InstanceTransform:"Transform (by ref)"=Get l_transform() -> (l_transform:Transform)
    bWorldSpace:bool=true
   ) -> (then, ReturnValue:int=0)
   > Set l_newIndex(l_newIndex:int=@n4) -> (Output_Get:int=0)
   > SetCustomDataValue!(
    execute
    self:Instanced Static Mesh Component Object Reference=Get PointsInstancedMesh()
    InstanceIndex:int=Get l_newIndex() -> (l_newIndex:int=0)
    CustomDataIndex:int=0
    CustomDataValue:float="To Float (Integer)"(InInt=Get l_newIndex() -> (l_newIndex=0))
    bMarkRenderStateDirty:bool=false
   ) -> (then, ReturnValue:bool)
   n5 = BreakColor(InColor=Get l_color() -> (l_color:Linear Color Structure))
   > SetCustomDataValue(
    self=Get PointsInstancedMesh()
    InstanceIndex=Get l_newIndex() -> (l_newIndex=0)
    CustomDataIndex=1
    CustomDataValue=@n5
    bMarkRenderStateDirty=false
   )
   > SetCustomDataValue(
    self=Get PointsInstancedMesh()
    InstanceIndex=Get l_newIndex() -> (l_newIndex=0)
    CustomDataIndex=2
    CustomDataValue=@n5.G
    bMarkRenderStateDirty=false
   )
   > SetCustomDataValue(
    self=Get PointsInstancedMesh()
    InstanceIndex=Get l_newIndex() -> (l_newIndex=0)
    CustomDataIndex=3
    CustomDataValue=@n5.B
    bMarkRenderStateDirty=false
   )
   > SetCustomDataValue(
    self=Get PointsInstancedMesh()
    InstanceIndex=Get l_newIndex() -> (l_newIndex=0)
    CustomDataIndex=4
    CustomDataValue=Get pointValue() -> (pointValue:double)
    bMarkRenderStateDirty=true
   )
   n6 = BreakTransform(InTransform=Get l_transform())
   n7 = AddInstance(
    execute=@n3.then_1
    self=Get FrustumInstancedMesh()
    InstanceTransform=MakeTransform(Location=@n6, Rotation=@n6.Rotation, Scale="1,1,0.5")
    bWorldSpace=true
   ) -> (ReturnValue=0)
   > Set l_newIndex(l_newIndex=@n7) -> (Output_Get=0)
   > SetCustomDataValue(
    self=Get FrustumInstancedMesh()
    InstanceIndex=Get l_newIndex() -> (l_newIndex=0)
    CustomDataIndex=0
    CustomDataValue="To Float (Integer)"(InInt=Get l_newIndex() -> (l_newIndex=0))
    bMarkRenderStateDirty=false
   )
   n8 = BreakColor(InColor=Get l_color())
   > SetCustomDataValue(
    self=Get FrustumInstancedMesh()
    InstanceIndex=Get l_newIndex() -> (l_newIndex=0)
    CustomDataIndex=1
    CustomDataValue=@n8
    bMarkRenderStateDirty=false
   )
   > SetCustomDataValue(
    self=Get FrustumInstancedMesh()
    InstanceIndex=Get l_newIndex() -> (l_newIndex=0)
    CustomDataIndex=2
    CustomDataValue=@n8.G
    bMarkRenderStateDirty=false
   )
   > SetCustomDataValue(
    self=Get FrustumInstancedMesh()
    InstanceIndex=Get l_newIndex() -> (l_newIndex=0)
    CustomDataIndex=3
    CustomDataValue=@n8.B
    bMarkRenderStateDirty=false
   )
   > SetCustomDataValue(
    self=Get FrustumInstancedMesh()
    InstanceIndex=Get l_newIndex() -> (l_newIndex=0)
    CustomDataIndex=4
    CustomDataValue=Get pointValue()
    bMarkRenderStateDirty=true
   )
  ]}
 clearInstances {nodes [
   clearInstances() {is=title K2Node_FunctionEntry}
   > ClearInstances(self:Instanced Static Mesh Component Object Reference=[Get FrustumInstancedMesh(), Get PointsInstancedMesh()])
  ]}
 setDrawEveryOther {nodes [
   n1 = setDrawEveryOther!() -> (then, everyOther:double=~) {is=title K2Node_FunctionEntry}
   > Set everyOther(everyOther:double=@n1) -> (Output_Get:double)
   > SetCustomPrimitiveDataFloat(self=[Get FrustumInstancedMesh(), Get PointsInstancedMesh()], DataIndex=2, Value=Get everyOther() -> (everyOther:double))
  ]}
 setFrustumForcedOn {nodes [
   n1 = setFrustumForcedOn!() -> (then, forcedOn:bool=~) {is=title K2Node_FunctionEntry}
   > Set allFrustumsForcedOn(allFrustumsForcedOn:bool=@n1) -> (Output_Get:bool)
   > validateFrustumVisual()
  ]}
 setFrustumRenderDistance {nodes [
   n1 = setFrustumRenderDistance!() -> (then, frustumDist:double=~) {is=title K2Node_FunctionEntry}
   > Set frustumDist(frustumDist:double=@n1) -> (Output_Get:double)
   > validateFrustumVisual()
  ]}
 setFrustumRenderOpacity {nodes [
   n1 = setFrustumRenderOpacity!() -> (then, frustumOpacity:double=~) {is=title K2Node_FunctionEntry}
   > Set frustumOpacity(frustumOpacity:double=@n1) -> (Output_Get:double)
   > validateFrustumVisual()
  ]}
 setFrustumRenderScale {nodes [
   n1 = setFrustumRenderScale!() -> (then, frustumDist:double=~) {is=title K2Node_FunctionEntry}
   > Set frustumScale(frustumScale:double=@n1) -> (Output_Get:double)
   > validateFrustumVisual()
  ]}
 setFrustumVisible {nodes [
   n1 = setFrustumVisible!() -> (then, visible:bool=~) {is=title K2Node_FunctionEntry}
   > Set isFrustumVisible(isFrustumVisible:bool=@n1) -> (Output_Get:bool)
   > validateFrustumVisual()
  ]}
 setScale {nodes [
   n1 = setScale!() -> (then, scale:double=~) {is=title K2Node_FunctionEntry}
   > Set scale(scale:double=@n1) -> (Output_Get:double)
   > SetCustomPrimitiveDataFloat(self=Get PointsInstancedMesh(), DataIndex=0, Value=Get scale() -> (scale:double))
  ]}
 setValueThreshold {nodes [
   n1 = setValueThreshold!() -> (then, threshold:double=~) {is=title K2Node_FunctionEntry}
   > Set valueThreshold(valueThreshold:double=@n1) -> (Output_Get:double)
   > SetCustomPrimitiveDataFloat(self=[Get FrustumInstancedMesh(), Get PointsInstancedMesh()], DataIndex=3, Value=Get valueThreshold() -> (valueThreshold:double))
  ]}
 setValueThresholdBool {nodes [
   n1 = setValueThresholdBool!() -> (then, hgherThanThreshold:bool=~) {is=title K2Node_FunctionEntry}
   > Set higherThanThreshold(higherThanThreshold:bool=@n1) -> (Output_Get:bool)
   > SetCustomPrimitiveDataFloat(
    self=[Get FrustumInstancedMesh(), Get PointsInstancedMesh()]
    DataIndex=4
    Value="To Float (Boolean)"(InBool=Get higherThanThreshold() -> (higherThanThreshold:bool))
   )
  ]}
 validateFrustumVisual {nodes [
   validateFrustumVisual() {is=title K2Node_FunctionEntry}
   > SetVisibility!(
    execute
    self:Scene Component Object Reference=Get FrustumInstancedMesh()
    bNewVisibility:bool=Get isFrustumVisible() -> (isFrustumVisible:bool)
    bPropagateToChildren:bool=false
   ) -> (then)
   > Branch(Condition=Get isFrustumVisible())
   > SetCustomPrimitiveDataFloat(self=Get FrustumInstancedMesh(), DataIndex=5, Value="To Float (Boolean)"(InBool=Get allFrustumsForcedOn() -> (allFrustumsForcedOn:bool)))
   > SetCustomPrimitiveDataFloat(self=Get FrustumInstancedMesh(), DataIndex=6, Value=float * float(A=Get frustumOpacity() -> (frustumOpacity:double), B=0.5))
   > SetCustomPrimitiveDataFloat(
    self=Get FrustumInstancedMesh()
    DataIndex=0
    Value=float * float(A=Get frustumScale() -> (frustumScale:double), B=Get frustumDist() -> (frustumDist:double))
   )
  ]}
]
variables [
 allFrustumsForcedOn {type=boolean, default=true}
 currentIndex {type=integer}
 everyOther {type=number}
 frustumDist {type=number, default=20000}
 frustumOpacity {type=number, default=0.5}
 frustumScale {type=number, default=1}
 higherThanThreshold {type=boolean}
 isFrustumVisible {type=boolean, default=true}
 scale {type=number, default=1}
 valueThreshold {type=number}
]
