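# Editor utility marker actor for the PerfCsvPlaybackTool, showing one indexed point with a color and an optional camera-frustum mesh.
# Setters drive materials, translucent sort priority, visibility and custom primitive data, highlighting the current marker.
# The frustum shows when enabled and either current or forced on, scaled from distance as a 60-degree, 16:9 view.
# setMyIndexSelected finds the tool's editor widget and sets its current index to this marker's.
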EditorUtilityBlueprint /Game/NeverCook/EditorUtilities/PerfTools/PerfCsvPlaybackTool/Actors/PCPT_DisplayUtilityActors
graphs [
 EventGraph {nodes [
   Event BeginPlay()
   > validateFrustumVisual()
  ]}
 UserConstructionScript {nodes [Construction Script()]}
 setColor {nodes [
   n1 = setColor!() -> (then, color:Linear Color Structure) {is=title K2Node_FunctionEntry}
   > Set color(color:Linear Color Structure=@n1) -> (Output_Get:Linear Color Structure)
   > SetCustomPrimitiveDataVector3!(
    execute
    self:Primitive Component Object Reference=[
     Get Frustum_SM() -> (Frustum_SM:Static Mesh Component Object Reference)
     Get StaticMesh!() -> (StaticMesh:Static Mesh Component Object Reference)
    ]
    DataIndex:int=1
    Value:Vector="To Vector (LinearColor)"!(InLinearColor:Linear Color Structure=Get color() -> (color:Linear Color Structure)) -> (ReturnValue:Vector)
   ) -> (then)
  ]}
 setCurrent {nodes [
   n1 = setCurrent!() -> (then, current:bool=~) {is=title K2Node_FunctionEntry}
   > Set current(current:bool=@n1) -> (Output_Get:bool)
   > SetCustomPrimitiveDataFloat!(
    execute
    self:Primitive Component Object Reference=Get StaticMesh()
    DataIndex:int=0
    Value:float="To Float (Boolean)"(InBool=Get current() -> (current:bool))
   ) -> (then)
   > validateVis()
   > SetTranslucentSortPriority!(
    execute
    self:Primitive Component Object Reference=Get StaticMesh()
    NewTranslucentSortPriority:int=SelectInt(A=0, B=999, bPickA=Get current())
   ) -> (then)
   > SetMaterial(
    self=Get StaticMesh()
    ElementIndex=0
    Material=Select(
     Option 0:Material Interface Object Reference=&/PCPT_Display_Mat
     Option 1:Material Interface Object Reference=&/PCPT_DisplayCurrent_Mat
     Index=Get current()
    ) -> (ReturnValue:Material Interface Object Reference)
   )
   > validateFrustumVisual()
   > SetIsTemporarilyHiddenInEditor!(execute, self:Actor Object Reference, bIsHidden:bool=false) -> (then)
  ]}
 setFrustumForcedOn {nodes [
   n1 = setFrustumForcedOn!() -> (then, forcedOn:bool=~) {is=title K2Node_FunctionEntry}
   > Set isFrustumForcedOn(isFrustumForcedOn:bool=@n1) -> (Output_Get:bool)
   > validateFrustumVisual()
  ]}
 setFrustumRenderDistance {nodes [
   n1 = setFrustumRenderDistance!() -> (then, frustumDist:double=~) {is=title K2Node_FunctionEntry}
   > Set frustumDist(frustumDist:double=@n1) -> (Output_Get:double)
   > validateFrustumVisual()
  ]}
 setFrustumRenderOpacity {nodes [
   n1 = setFrustumRenderOpacity!() -> (then, frustumOpacity:double=~) {is=title K2Node_FunctionEntry}
   > Set frustumOpacity(frustumOpacity:double=@n1) -> (Output_Get:double)
   > validateFrustumVisual()
  ]}
 setFrustumRenderScale {nodes [
   n1 = setFrustumRenderScale!() -> (then, frustumDist:double=~) {is=title K2Node_FunctionEntry}
   > Set frustumScale(frustumScale:double=@n1) -> (Output_Get:double)
   > validateFrustumVisual()
  ]}
 setFrustumVisible {nodes [
   n1 = setFrustumVisible!() -> (then, visible:bool=~) {is=title K2Node_FunctionEntry}
   > Set isFrustumVisible(isFrustumVisible:bool=@n1) -> (Output_Get:bool)
   > validateFrustumVisual()
  ]}
 setIsVisible {nodes [
   n1 = setIsVisible!() -> (then, isVisible:bool=~) {is=title K2Node_FunctionEntry}
   > Set isVisible(isVisible:bool=@n1) -> (Output_Get:bool)
   > validateVis()
  ]}
 setMyIndexSelected {nodes [
   setMyIndexSelected() {is=title K2Node_FunctionEntry}
   > n1 = FindUtilityWidgetFromBlueprint!(
    execute
    self:Editor Utility Subsystem Object Reference=Get EditorUtilitySubsystem()
    InBlueprint:Editor Utility Widget Blueprint Object Reference=/Game/NeverCook/EditorUtilities/PerfTools/PerfCsvPlaybackTool/PerfCsvPlaybackTool_EWidget
   ) -> (then, ReturnValue:Editor Utility Widget Object Reference)
   > n2 = Cast To PerfCsvPlaybackTool_EWidget!(execute, Object=@n1) -> (then, CastFailed, AsPerf Csv Playback Tool EWidget:Perf Csv Playback Tool EWidget Object Reference) {is=title K2Node_DynamicCast}
   > setCurrentIndex!(execute, self:Perf Csv Playback Tool EWidget Object Reference=@n2, currentIndex:int=Get index() -> (index:int=0), updateTimeline:bool=true) -> (then)
  ]}
 setScale {nodes [
   n1 = setScale!() -> (then, scale:double=~) {is=title K2Node_FunctionEntry}
   > n2 = Set scale(scale:double=@n1) -> (Output_Get:double)
   > SetRelativeScale3D!(
    execute
    self:Scene Component Object Reference=Get ScaleScene() -> (ScaleScene:Scene Component Object Reference)
    NewScale3D:Vector="To Vector (Float)"!(InDouble:double=@n2) -> (ReturnValue:Vector)
   ) -> (then)
  ]}
 validateFrustumVisual {nodes [
   validateFrustumVisual() {is=title K2Node_FunctionEntry}
   n1 = Get current()
   n2 = Get isFrustumForcedOn() -> (isFrustumForcedOn:bool)
   > SetVisibility!(
    execute
    self:Scene Component Object Reference=Get Frustum_SM()
    bNewVisibility:bool=AND Boolean(A=OR Boolean(A=@n1, B=@n2), B=Get isFrustumVisible() -> (isFrustumVisible:bool))
    bPropagateToChildren:bool=false
   ) -> (then)
   n3 = float * float!(
    A=Get frustumDist() -> (frustumDist:double)
    B="Tan (Degrees)"!(A:double=30) -> (ReturnValue:double)
    C:double=Get frustumScale() -> (frustumScale:double)
   ) -> (ReturnValue)
   n4 = float * float!(A=@n3, B=0.5625) -> (ReturnValue)
   n5 = Get frustumDist()
   n6 = Get Frustum_SM()
   > SetRelativeScale3D(
    self=Get Frustum_SM()
    NewScale3D=SelectVector(A=MakeVector(X=@n5, Y=@n3, Z=@n4), B="0, 0, 0", bPickA=IsVisible(self:Scene Component Object Reference=@n6))
   )
   > SetCustomPrimitiveDataFloat(self=Get Frustum_SM(), DataIndex=5, Value=Get frustumOpacity() -> (frustumOpacity:double))
  ]}
 validateVis {nodes [
   validateVis() {is=title K2Node_FunctionEntry}
   n1 = Get isVisible() -> (isVisible:bool)
   n2 = Get current()
   > Set Actor Hidden In Game!(execute, self:Actor Object Reference, bNewHidden:bool=NOT Boolean(A=OR Boolean(A=@n1, B=@n2))) -> (then) {is=title K2Node_CallFunction}
  ]}
]
variables [
 color {type=LinearColor, default=<rgba 0 0 0 0>}
 current {type=boolean}
 frustumDist {type=number, default=20000}
 frustumOpacity {type=number, default=0.5}
 frustumScale {type=number, default=1}
 index {type=integer}
 isFrustumForcedOn {type=boolean}
 isFrustumVisible {type=boolean, default=true}
 isVisible {type=boolean}
 scale {type=number}
]
