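# Editor utility widget for the perf CSV playback tool: a collapsible panel setting the location and rotation offset for drawing data in world.
# Spin boxes take typed values, the selected actor's transform, or pasted clipboard text, exposed through getter and setter functions.
# handleOffsetUpdated refreshes clear buttons and status text, optionally broadcasting onTransformUpdated.
# Paste reads the clipboard through Python 2-style Tkinter script, parsing its log output.
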EditorUtilityWidgetBlueprint /Game/NeverCook/EditorUtilities/PerfTools/PerfCsvPlaybackTool/PCPT_OffsetTransform_EWidget
graphs [
 EventGraph {nodes [
   n1 = "On Value Committed (spinX)"!() -> (OutputDelegate, then, InValue:float, CommitMethod:ETextCommit Enum=Default) {is=title K2Node_ComponentBoundEvent}
   handleOffsetUpdated!(execute=[@n1.then, @n2.then, @n3.then], self, throwEvent:bool=true) -> (then)
   n2 = "On Value Committed (spinY)"!() -> (OutputDelegate, then, InValue:float, CommitMethod:ETextCommit Enum=Default) {is=title K2Node_ComponentBoundEvent}
   n3 = "On Value Committed (spinZ)"!() -> (OutputDelegate, then, InValue:float, CommitMethod:ETextCommit Enum=Default) {is=title K2Node_ComponentBoundEvent}
   "On Clicked (btnPasteLoc)"!() -> (OutputDelegate, then, widget:User Widget Object Reference, event:Pointer Event Structure) {is=title K2Node_ComponentBoundEvent}
   > pasteLocation()
   "On Clicked (btnClearLoc)"!() -> (OutputDelegate, then, widget:User Widget Object Reference, event:Pointer Event Structure) {is=title K2Node_ComponentBoundEvent}
   > setLocationOffset!(execute, self, location:Vector="0, 0, 0") -> (then)
   > handleOffsetUpdated(throwEvent=true)
   n4 = "On Value Committed (spinRoll)"!() -> (OutputDelegate, then, InValue:float, CommitMethod:ETextCommit Enum=Default) {is=title K2Node_ComponentBoundEvent}
   handleOffsetUpdated(execute=[@n4.then, @n9.then, @n10.then], throwEvent=true)
   "On Clicked (btnClearRot)"!() -> (OutputDelegate, then, widget:User Widget Object Reference, event:Pointer Event Structure) {is=title K2Node_ComponentBoundEvent}
   > setRotationOffset!(execute, self, location:Rotator="0, 0, 0") -> (then)
   > handleOffsetUpdated(throwEvent=true)
   "On Clicked (btnGetFromSelection)"!() -> (OutputDelegate, then, widget:User Widget Object Reference, event:Pointer Event Structure) {is=title K2Node_ComponentBoundEvent}
   > n5 = GetSelectedActors()
   > Branch(Condition=Is Valid Index!(TargetArray:Array of Actor Object References=@n5, IndexToTest:int) -> (ReturnValue:bool))
   n6 = Get Actor Transform!(self="Get (a copy)"(Array:Array of Actor Object References=@n5) -> (Output:Actor Object Reference)) -> (ReturnValue_Location:Vector, ReturnValue_Rotation:Rotator, ReturnValue_Scale:Vector)
   n7 = BreakVector(InVec=@n6)
   > SetValue!(execute, self:Spin Box Object Reference=Get spinX() -> (spinX:Spin Box Object Reference), NewValue:float=@n7) -> (then)
   > SetValue(self=Get spinY() -> (spinY:Spin Box Object Reference), NewValue=@n7.Y)
   > SetValue(self=Get spinZ() -> (spinZ:Spin Box Object Reference), NewValue=@n7.Z)
   n8 = BreakRotator(InRot=@n6.ReturnValue_Rotation)
   > SetValue(self=Get spinRoll() -> (spinRoll:Spin Box Object Reference), NewValue=@n8)
   > SetValue(self=Get spinPitch() -> (spinPitch:Spin Box Object Reference), NewValue=@n8.Pitch)
   > SetValue(self=Get spinYaw() -> (spinYaw:Spin Box Object Reference), NewValue=@n8.Yaw)
   > handleOffsetUpdated(throwEvent=true)
   n9 = "On Value Committed (spinPitch)"!() -> (OutputDelegate, then, InValue:float, CommitMethod:ETextCommit Enum=Default) {is=title K2Node_ComponentBoundEvent}
   n10 = "On Value Committed (spinYaw)"!() -> (OutputDelegate, then, InValue:float, CommitMethod:ETextCommit Enum=Default) {is=title K2Node_ComponentBoundEvent}
   "On Clicked (btnPasteRot)"!() -> (OutputDelegate, then, widget:User Widget Object Reference, event:Pointer Event Structure) {is=title K2Node_ComponentBoundEvent}
   > pasteRotation()
   "On Clicked (btnClearTransform)"!() -> (OutputDelegate, then, widget:User Widget Object Reference, event:Pointer Event Structure) {is=title K2Node_ComponentBoundEvent}
   > setLocationOffset(location="0, 0, 0")
   > setRotationOffset(location="0, 0, 0")
   > handleOffsetUpdated(throwEvent=true)
   Event Construct()
   > Parent: Construct()
   > handleOffsetUpdated(throwEvent=false)
   Event Destruct()
   > Parent: Destruct()
   n11 = Event Tick()
   > Parent: Tick!(execute, MyGeometry:Geometry Structure=@n11.MyGeometry, InDeltaTime:float=@n11.InDeltaTime) -> (then) {is=title K2Node_CallParentFunction}
   n12 = Event PreConstruct()
   > Parent: PreConstruct(IsDesignTime=@n12.IsDesignTime)
   > SetIsExpanded!(
    execute
    self:Expandable Area Object Reference=Get ExpandableArea_67() -> (ExpandableArea_67:Expandable Area Object Reference)
    IsExpanded:bool=false
   ) -> (then)
  ]}
 getLocationOffset {nodes [
   getLocationOffset() {is=title K2Node_FunctionEntry}
   n1 = Get spinZ()
   n2 = Get spinX()
   n3 = Get spinY()
   > Return Node!(
    execute
    ReturnValue:Vector=MakeVector(X=GetValue!(self:Spin Box Object Reference=@n2) -> (ReturnValue:float), Y=GetValue(self=@n3), Z=GetValue(self=@n1))
   )
  ]}
 getRotationOffset {nodes [
   getRotationOffset() {is=title K2Node_FunctionEntry}
   n1 = Get spinYaw()
   n2 = Get spinRoll()
   n3 = Get spinPitch()
   > Return Node(ReturnValue:Rotator=MakeRotator(Roll=GetValue(self=@n2), Pitch=GetValue(self=@n3), Yaw=GetValue(self=@n1)))
  ]}
 handleOffsetUpdated {nodes [
   n1 = handleOffsetUpdated!() -> (then, throwEvent:bool=true) {is=title K2Node_FunctionEntry}
   > Set l_locationChanged(l_locationChanged:bool="Not Equal (Vector)"!(A:Vector=getLocationOffset!() -> (ReturnValue:Vector) {is=title K2Node_CallFunction}, B:Vector="0, 0, 0", ErrorTolerance:float=0.0001) -> (ReturnValue:bool)) -> (Output_Get:bool)
   > Set l_rotationChanged(l_rotationChanged:bool="Not Equal (Rotator)"!(A:Rotator=getRotationOffset!() -> (ReturnValue:Rotator) {is=title K2Node_CallFunction}, B:Rotator="0, 0, 0", ErrorTolerance:float=0.0001) -> (ReturnValue:bool)) -> (Output_Get:bool)
   > n2 = Sequence!(execute) -> (then_0, then_1, then_2:Exec, then_3:Exec, then_4:Exec)
   > SetVisibility(
    self=Get btnClearLoc() -> (btnClearLoc:TA Generic 01 Button Object Reference)
    InVisibility=Select(Option 0=Hidden, Option 1=Visible, Index=Get l_locationChanged() -> (l_locationChanged:bool))
   )
   SetVisibility(
    execute=@n2.then_1
    self=Get btnClearRot() -> (btnClearRot:TA Generic 01 Button Object Reference)
    InVisibility=Select(Option 0=Hidden, Option 1=Visible, Index=Get l_rotationChanged() -> (l_rotationChanged:bool))
   )
   n3 = Get l_rotationChanged()
   n4 = Get l_locationChanged()
   SetVisibility(
    execute=@n2.then_2
    self=Get btnClearTransform() -> (btnClearTransform:TA Generic 01 Button Object Reference)
    InVisibility=Select(Option 0=Hidden, Option 1=Visible, Index=OR Boolean(A=@n4, B=@n3))
   )
   Branch(execute=@n2.then_3, Condition=@n1)
   > Call On Transform Updated!(execute:Exec, self:PCPT Offset Transform EWidget Object Reference) -> (then:Exec) {is=title K2Node_CallDelegate}
   n5 = Get l_locationChanged()
   n6 = Get l_rotationChanged()
   "SetText (Text)"(
    execute=@n2.then_4
    self=Get txtTransform() -> (txtTransform:Common Text Object Reference)
    InText=Select(
     Option 0:Text="LOCTEXT(\"No Transform Offset Applied\")"
     Option 1:Text="LOCTEXT(\"Transform Offset is Applied\")"
     Index=OR Boolean(A=@n5, B=@n6)
    ) -> (ReturnValue:Text)
   )
   n7 = OR Boolean(A=Get l_locationChanged(), B=Get l_rotationChanged())
   > SetColorAndOpacity(
    self=Get txtTransform()
    InColorAndOpacity=Make Slate Color(SpecifiedColor=SelectColor(A=(R=0.8958,G=0.2901,B=0.0458,A=1), B=(R=0.9167,G=0.9167,B=0.9167,A=1), bPickA=@n7))
   )
  ]}
 onTransformUpdated {nodes [onTransformUpdated() {is=title K2Node_FunctionEntry}]}
 pasteLocation {nodes [
   pasteLocation() {is=title K2Node_FunctionEntry}
   > n1 = "Execute Python Command (Advanced)"!(
    execute
    PythonCommand:String="from Tkinter import *\r\nroot = Tk()\r\ncliptext = root.clipboard_get()\r\nprint cliptext\r\nroot.withdraw()"
    ExecutionMode:EPythonCommandExecutionMode Enum=ExecuteFile
    FileExecutionScope:EPythonFileExecutionScope Enum=Private
   ) -> (then, CommandResult:String, LogOutput:Array of Python Log Output Entry Structures, ReturnValue:bool)
   > Branch(Condition=Is Valid Index(TargetArray:Array of Python Log Output Entry Structures=@n1.LogOutput))
   n2 = Break Python Log Output Entry!(PythonLogOutputEntry:"Python Log Output Entry Structure (by ref)"="Get (a copy)"(Array:Array of Python Log Output Entry Structures=@n1.LogOutput) -> (Output:Python Log Output Entry Structure)) -> (Type:EPythonLogOutputType Enum, Output:String) {is=title K2Node_BreakStruct}
   > Branch(Condition=AND Boolean!(
     A=Contains(SearchIn=@n2.Output, Substring="X=")
     B=Contains(SearchIn=@n2.Output, Substring="Y=")
     C:bool=Contains(SearchIn=@n2.Output, Substring="Z=")
    ) -> (ReturnValue))
   n3 = Replace(SourceString=@n2.Output, From=")")
   > n4 = For Each Loop(Array=ParseIntoArray(SourceString=Replace(SourceString=@n3, From="("), Delimiter=","))
   n5 = Get l_floats() -> (l_floats:"Array of Float (double-precision)s")
   Branch(
    execute=@n4.Completed
    Condition=AND Boolean(
     A=Is Valid Index(TargetArray:"Array of Float (double-precision)s"=@n5)
     B=Is Valid Index(TargetArray=@n5, IndexToTest=1)
     C=Is Valid Index(TargetArray=@n5, IndexToTest=2)
    )
   )
   n6 = Get l_floats()
   > SetValue(self=Get spinX(), NewValue="Get (a copy)"(Array:"Array of Float (double-precision)s"=@n6) -> (Output:double))
   > SetValue(self=Get spinY(), NewValue="Get (a copy)"(Array=@n6, Dimension 1=1))
   > SetValue(self=Get spinZ(), NewValue="Get (a copy)"(Array=@n6, Dimension 1=2))
   > handleOffsetUpdated(throwEvent=true) {is=title K2Node_CallFunction}
   n7 = Split(SourceString=@n4, InStr="=", SearchDir=FromStart)
   Branch(execute=@n4.LoopBody, Condition=@n7.ReturnValue)
   > Add(
    TargetArray:"Array of Float (double-precision)s"=Get l_floats()
    NewItem:"Float (double-precision) (by ref)"=String To Float(InString=@n7.RightS)
   ) {is=title K2Node_CallArrayFunction}
  ]}
 pasteRotation {nodes [
   pasteRotation() {is=title K2Node_FunctionEntry}
   > n1 = "Execute Python Command (Advanced)"(
    PythonCommand="from Tkinter import *\r\nroot = Tk()\r\ncliptext = root.clipboard_get()\r\nprint cliptext\r\nroot.withdraw()"
    ExecutionMode=ExecuteFile
    FileExecutionScope=Private
   )
   > Branch(Condition=Is Valid Index(TargetArray:Array of Python Log Output Entry Structures=@n1.LogOutput))
   n2 = Break Python Log Output Entry(PythonLogOutputEntry="Get (a copy)"(Array:Array of Python Log Output Entry Structures=@n1.LogOutput) -> (Output:Python Log Output Entry Structure))
   > Branch(Condition=AND Boolean(
     A=Contains(SearchIn=@n2.Output, Substring="Pitch=")
     B=Contains(SearchIn=@n2.Output, Substring="Yaw=")
     C=Contains(SearchIn=@n2.Output, Substring="Roll=")
    ))
   n3 = Replace(SourceString=@n2.Output, From=")")
   > n4 = For Each Loop(Array=ParseIntoArray(SourceString=Replace(SourceString=@n3, From="("), Delimiter=","))
   n5 = Get l_floats()
   Branch(
    execute=@n4.Completed
    Condition=AND Boolean(
     A=Is Valid Index(TargetArray:"Array of Float (double-precision)s"=@n5)
     B=Is Valid Index(TargetArray=@n5, IndexToTest=1)
     C=Is Valid Index(TargetArray=@n5, IndexToTest=2)
    )
   )
   n6 = Get l_floats()
   > SetValue(self=Get spinRoll(), NewValue="Get (a copy)"(Array:"Array of Float (double-precision)s"=@n6) -> (Output:double))
   > SetValue(self=Get spinPitch(), NewValue="Get (a copy)"(Array=@n6, Dimension 1=1))
   > SetValue(self=Get spinYaw(), NewValue="Get (a copy)"(Array=@n6, Dimension 1=2))
   > handleOffsetUpdated(throwEvent=true)
   n7 = Split(SourceString=@n4, InStr="=", SearchDir=FromStart)
   Branch(execute=@n4.LoopBody, Condition=@n7.ReturnValue)
   > Add(TargetArray=Get l_floats(), NewItem=String To Float(InString=@n7.RightS))
  ]}
 setLocationOffset {nodes [
   n1 = setLocationOffset!() -> (then, location:Vector=~) {is=title K2Node_FunctionEntry}
   n2 = BreakVector(InVec=@n1)
   > SetValue(self=Get spinX(), NewValue=@n2)
   > SetValue(self=Get spinY(), NewValue=@n2.Y)
   > SetValue(self=Get spinZ(), NewValue=@n2.Z)
   > handleOffsetUpdated(throwEvent=false)
  ]}
 setRotationOffset {nodes [
   n1 = setRotationOffset!() -> (then, location:Rotator=~) {is=title K2Node_FunctionEntry}
   n2 = BreakRotator(InRot=@n1)
   > SetValue(self=Get spinRoll(), NewValue=@n2)
   > SetValue(self=Get spinPitch(), NewValue=@n2.Pitch)
   > SetValue(self=Get spinYaw(), NewValue=@n2.Yaw)
   > handleOffsetUpdated(throwEvent=false)
  ]}
]
variables [
 OffsetTransform {type=Transform, default {location=<0 0 0>, rotation=<rot 0 0 0>, scale=<1 1 1>}}
 onTransformUpdated {default=()}
]
widgets {
 info {
  namedSlotCount=2
  parentClass=/Game/Blueprints/Blutilities/ForTechArt/Widgets/TA_Base_EditorWidget.TA_Base_EditorWidget_C
  rootWidgetClass=&VerticalBox
  widgetCount=30
 }
 widgets [
  VerticalBox_0 {widgetClassPath=&VerticalBox, widget=VerticalBox_0}
  Border_2 {
   widgetClassPath=&Border
   parent=VerticalBox_0
   widget=Border_2
   properties {
    background {
     bIsDynamicallyLoaded=false
     drawAs=Image
     imageType=NoImage
     mirroring=NoMirror
     resourceName=None
     resourceObject=null
     tiling=NoTile
     imageSize=<32 32>
     margin=<ltrb 0 0 0 0>
     outlineSettings {
      bUseBrushTransparency=false
      roundingType=HalfHeightRadius
      width=0
      color {colorUseRule=UseColor_Specified, specifiedColor=<rgba 0 0 0 0>}
      cornerRadii=<0 0 0 0>
     }
     tintColor {colorUseRule=UseColor_Specified, specifiedColor=<rgba 0.0627 0.1412 0.2392 1>}
     uVRegion {bIsValid=false, max=<0 0>, min=<0 0>}
    }
    padding=<ltrb 0 0 0 0>
   }
   slot {class=VerticalBoxSlot}
  }
  Image_134 {
   widgetClassPath=&Image
   bIsVariable=true
   parent=Border_2
   widget=Image_134
   properties {
    brush {
     bIsDynamicallyLoaded=false
     drawAs=Image
     imageType=NoImage
     mirroring=NoMirror
     resourceName=None
     resourceObject=null
     tiling=NoTile
     imageSize=<32 4>
     margin=<ltrb 0 0 0 0>
     outlineSettings {
      bUseBrushTransparency=false
      roundingType=HalfHeightRadius
      width=0
      color {colorUseRule=UseColor_Specified, specifiedColor=<rgba 0 0 0 0>}
      cornerRadii=<0 0 0 0>
     }
     tintColor {colorUseRule=UseColor_Specified, specifiedColor=<rgba 1 1 1 1>}
     uVRegion {bIsValid=false, max=<0 0>, min=<0 0>}
    }
    colorAndOpacity=<rgba 0.0205 0.0548 0.0938 1>
   }
   slot {class=BorderSlot, properties {padding=<ltrb 0 0 0 0>}}
  }
  ExpandableArea_67 {
   widgetClassPath=/Script/UMG.ExpandableArea
   bIsVariable=true
   parent=VerticalBox_0
   widget=ExpandableArea_67
   properties {
    borderBrush {
     bIsDynamicallyLoaded=false
     drawAs=RoundedBox
     imageType=NoImage
     mirroring=NoMirror
     resourceName=None
     resourceObject=/Game/ContentCreationTools/ObjectPlacement/Textures/T_BlankWhite_01
     tiling=NoTile
     imageSize=<4 4>
     margin=<ltrb 0 0 0 0>
     outlineSettings {
      bUseBrushTransparency=false
      roundingType=FixedRadius
      width=0
      color {colorUseRule=UseColor_Specified, specifiedColor=<rgba 0 0 0 0>}
      cornerRadii=<4 4 4 4>
     }
     tintColor {colorUseRule=UseColor_Specified, specifiedColor=<rgba 0.0176 0.0176 0.0176 1>}
     uVRegion {bIsValid=false, max=<0 0>, min=<0 0>}
    }
    borderColor {colorUseRule=UseColor_Specified, specifiedColor=<rgba 0.0649 0.1443 0.2396 1>}
    style {
     rolloutAnimationSeconds=0.1
     collapsedImage {
      bIsDynamicallyLoaded=false
      drawAs=Image
      imageType=Vector
      mirroring=NoMirror
      resourceName=../../../Engine/Content/Slate/Starship/Common/chevron-right.svg
      resourceObject=null
      tiling=NoTile
      imageSize=<20 20>
      margin=<ltrb 0 0 0 0>
      outlineSettings {
       bUseBrushTransparency=false
       roundingType=HalfHeightRadius
       width=0
       color {colorUseRule=UseColor_Specified, specifiedColor=<rgba 0 0 0 0>}
       cornerRadii=<0 0 0 0>
      }
      tintColor {colorUseRule=UseColor_Specified, specifiedColor=<rgba 0.5271 0.5271 0.5271 1>}
      uVRegion {bIsValid=false, max=<0 0>, min=<0 0>}
     }
     expandedImage {
      bIsDynamicallyLoaded=false
      drawAs=Image
      imageType=Vector
      mirroring=NoMirror
      resourceName=../../../Engine/Content/Slate/Starship/Common/chevron-down.svg
      resourceObject=null
      tiling=NoTile
      imageSize=<20 20>
      margin=<ltrb 0 0 0 0>
      outlineSettings {
       bUseBrushTransparency=false
       roundingType=HalfHeightRadius
       width=0
       color {colorUseRule=UseColor_Specified, specifiedColor=<rgba 0 0 0 0>}
       cornerRadii=<0 0 0 0>
      }
      tintColor {colorUseRule=UseColor_Specified, specifiedColor=<rgba 0.5271 0.5271 0.5271 1>}
      uVRegion {bIsValid=false, max=<0 0>, min=<0 0>}
     }
    }
   }
   slot {class=VerticalBoxSlot}
  }
  HorizontalBox_171 {widgetClassPath=&HorizontalBox, namedSlotHost=ExpandableArea_67, widget=HorizontalBox_171}
  txtTransform {
   widgetClassPath=&CommonTextBlock
   bIsVariable=true
   parent=HorizontalBox_171
   widget=txtTransform
   properties {
    justification=Center
    text=Draw Transform Offset
    toolTipText="When drawing the data in world you can use this to offset the location we draw from.  If this is the transform of a Foundation in Apollo_Terrain then running this in the POI level will make the data line up properly."
    visibility=Visible
    font {
     bForceMonospaced=false
     bMaterialIsStencil=false
     fontMaterial=null
     fontObject=&NotoSans
     letterSpacing=0
     monospacedWidth=1
     size=12
     skewAmount=0
     typefaceFontName=NotoSans-Regular
     outlineSettings {
      bApplyOutlineToDropShadows=false
      bMiteredCorners=false
      bSeparateFillAlpha=false
      outlineMaterial=null
      outlineSize=0
      outlineColor=<rgba 0 0 0 1>
     }
    }
    shadowOffset=<0 0>
   }
   slot {class=HorizontalBoxSlot, properties {padding=<ltrb 0 1 0 0>, size {sizeRule=Fill, value=1}}}
  }
  btnGetFromSelection {
   widgetClassPath=&TA_Generic01_Button
   parent=HorizontalBox_171
   widget=btnGetFromSelection
   properties {label=Get Offset From Selected Actor, buttonColor_Default=<rgba 0.2782 0.6979 0.0962 1>}
   slot {class=HorizontalBoxSlot, properties {horizontalAlignment=HAlign_Right, verticalAlignment=VAlign_Center, padding=<ltrb 5 0 0 0>}}
  }
  btnClearTransform {
   widgetClassPath=&TA_Generic01_Button
   parent=HorizontalBox_171
   widget=btnClearTransform
   properties {label=Clear, toolTipText=Clear all offsets, buttonColor_Default=<rgba 1 0.1042 0.1042 1>}
   slot {class=HorizontalBoxSlot, properties {horizontalAlignment=HAlign_Right, verticalAlignment=VAlign_Center, padding=<ltrb 5 0 0 0>}}
  }
  Border_0 {
   widgetClassPath=&Border
   namedSlotHost=ExpandableArea_67
   widget=Border_0
   properties {brushColor=<rgba 1 1 1 0.183>, padding=<ltrb 5 5 5 5>}
  }
  VerticalBox_2 {widgetClassPath=&VerticalBox, parent=Border_0, widget=VerticalBox_2, slot {class=BorderSlot, properties {padding=<ltrb 5 5 5 5>}}}
  HorizontalBox_94 {
   widgetClassPath=&HorizontalBox
   parent=VerticalBox_2
   widget=HorizontalBox_94
   slot {class=VerticalBoxSlot, properties {size {sizeRule=Fill, value=1}}}
  }
  textField_0 {
   widgetClassPath=&CommonTextBlock
   parent=HorizontalBox_94
   widget=textField_0
   properties {
    justification=Right
    minDesiredWidth=150
    text=Location Offset:
    visibility=Visible
    font {
     bForceMonospaced=false
     bMaterialIsStencil=false
     fontMaterial=null
     fontObject=&NotoSans
     letterSpacing=0
     monospacedWidth=1
     size=12
     skewAmount=0
     typefaceFontName=NotoSans-Regular
     outlineSettings {
      bApplyOutlineToDropShadows=false
      bMiteredCorners=false
      bSeparateFillAlpha=false
      outlineMaterial=null
      outlineSize=0
      outlineColor=<rgba 0 0 0 1>
     }
    }
    shadowOffset=<0 0>
   }
   slot {class=HorizontalBoxSlot, properties {verticalAlignment=VAlign_Center}}
  }
  txtMin_4 {
   widgetClassPath=&CommonTextBlock
   parent=HorizontalBox_94
   widget=txtMin_4
   properties {
    text=X:
    visibility=Visible
    font {
     bForceMonospaced=false
     bMaterialIsStencil=false
     fontMaterial=null
     fontObject=&NotoSans
     letterSpacing=0
     monospacedWidth=1
     size=10
     skewAmount=0
     typefaceFontName=NotoSans-Regular
     outlineSettings {
      bApplyOutlineToDropShadows=false
      bMiteredCorners=false
      bSeparateFillAlpha=false
      outlineMaterial=null
      outlineSize=0
      outlineColor=<rgba 0 0 0 1>
     }
    }
    shadowOffset=<0 0>
   }
   slot {class=HorizontalBoxSlot, properties {verticalAlignment=VAlign_Center, padding=<ltrb 5 0 3 0>}}
  }
  spinX {
   widgetClassPath=/Script/UMG.SpinBox
   bIsVariable=true
   parent=HorizontalBox_94
   widget=spinX
   properties {
    bAlwaysUsesDeltaSnap=true
    delta=0.01
    maxFractionalDigits=3
    maxSliderValue=5000
    maxValue=5000
    minDesiredWidth=90
    minSliderValue=1000
    foregroundColor {colorUseRule=UseColor_Specified, specifiedColor=<rgba 1 1 1 1>}
    widgetStyle {
     activeBackgroundBrush {
      bIsDynamicallyLoaded=false
      drawAs=RoundedBox
      imageType=NoImage
      mirroring=NoMirror
      resourceName=None
      resourceObject=null
      tiling=NoTile
      imageSize=<0 0>
      margin=<ltrb 0 0 0 0>
      outlineSettings {
       bUseBrushTransparency=false
       roundingType=FixedRadius
       width=1
       color {colorUseRule=UseColor_ColorTable, specifiedColor=<rgba 0 0 0 0>}
       cornerRadii=<4 4 4 4>
      }
      tintColor {colorUseRule=UseColor_Specified, specifiedColor=<rgba 0.0048 0.0048 0.0048 1>}
      uVRegion {bIsValid=false, max=<0 0>, min=<0 0>}
     }
     activeFillBrush {
      bIsDynamicallyLoaded=false
      drawAs=RoundedBox
      imageType=NoImage
      mirroring=NoMirror
      resourceName=None
      resourceObject=null
      tiling=NoTile
      imageSize=<0 0>
      margin=<ltrb 0 0 0 0>
      outlineSettings {
       bUseBrushTransparency=false
       roundingType=FixedRadius
       width=1
       color {colorUseRule=UseColor_Specified, specifiedColor=<rgba 0 0 0 0>}
       cornerRadii=<4 4 4 4>
      }
      tintColor {colorUseRule=UseColor_Specified, specifiedColor=<rgba 0 0.162 0.7454 1>}
      uVRegion {bIsValid=false, max=<0 0>, min=<0 0>}
     }
     arrowsImage {
      bIsDynamicallyLoaded=false
      drawAs=NoDrawType
      imageType=NoImage
      mirroring=NoMirror
      resourceName=None
      resourceObject=null
      tiling=NoTile
      imageSize=<0 0>
      margin=<ltrb 0 0 0 0>
      outlineSettings {
       bUseBrushTransparency=false
       roundingType=HalfHeightRadius
       width=0
       color {colorUseRule=UseColor_Specified, specifiedColor=<rgba 0 0 0 0>}
       cornerRadii=<0 0 0 0>
      }
      tintColor {colorUseRule=UseColor_Specified, specifiedColor=<rgba 1 1 1 1>}
      uVRegion {bIsValid=false, max=<0 0>, min=<0 0>}
     }
     backgroundBrush {
      bIsDynamicallyLoaded=false
      drawAs=RoundedBox
      imageType=NoImage
      mirroring=NoMirror
      resourceName=None
      resourceObject=null
      tiling=NoTile
      imageSize=<0 0>
      margin=<ltrb 0 0 0 0>
      outlineSettings {
       bUseBrushTransparency=false
       roundingType=FixedRadius
       width=1
       color {colorUseRule=UseColor_ColorTable, specifiedColor=<rgba 0 0 0 0>}
       cornerRadii=<4 4 4 4>
      }
      tintColor {colorUseRule=UseColor_Specified, specifiedColor=<rgba 0.0048 0.0048 0.0048 1>}
      uVRegion {bIsValid=false, max=<0 0>, min=<0 0>}
     }
     foregroundColor {colorUseRule=UseColor_Specified, specifiedColor=<rgba 1 1 1 1>}
     hoveredBackgroundBrush {
      bIsDynamicallyLoaded=false
      drawAs=RoundedBox
      imageType=NoImage
      mirroring=NoMirror
      resourceName=None
      resourceObject=null
      tiling=NoTile
      imageSize=<0 0>
      margin=<ltrb 0 0 0 0>
      outlineSettings {
       bUseBrushTransparency=false
       roundingType=FixedRadius
       width=1
       color {colorUseRule=UseColor_ColorTable, specifiedColor=<rgba 0 0 0 0>}
       cornerRadii=<4 4 4 4>
      }
      tintColor {colorUseRule=UseColor_Specified, specifiedColor=<rgba 0.0048 0.0048 0.0048 1>}
      uVRegion {bIsValid=false, max=<0 0>, min=<0 0>}
     }
     hoveredFillBrush {
      bIsDynamicallyLoaded=false
      drawAs=RoundedBox
      imageType=NoImage
      mirroring=NoMirror
      resourceName=None
      resourceObject=null
      tiling=NoTile
      imageSize=<0 0>
      margin=<ltrb 0 0 0 0>
      outlineSettings {
       bUseBrushTransparency=false
       roundingType=FixedRadius
       width=1
       color {colorUseRule=UseColor_Specified, specifiedColor=<rgba 0 0 0 0>}
       cornerRadii=<4 4 4 4>
      }
      tintColor {colorUseRule=UseColor_Specified, specifiedColor=<rgba 0.0953 0.0953 0.0953 1>}
      uVRegion {bIsValid=false, max=<0 0>, min=<0 0>}
     }
     inactiveFillBrush {
      bIsDynamicallyLoaded=false
      drawAs=RoundedBox
      imageType=NoImage
      mirroring=NoMirror
      resourceName=None
      resourceObject=null
      tiling=NoTile
      imageSize=<0 0>
      margin=<ltrb 0 0 0 0>
      outlineSettings {
       bUseBrushTransparency=false
       roundingType=FixedRadius
       width=1
       color {colorUseRule=UseColor_Specified, specifiedColor=<rgba 0 0 0 0>}
       cornerRadii=<4 4 4 4>
      }
      tintColor {colorUseRule=UseColor_Specified, specifiedColor=<rgba 0.0395 0.0395 0.0395 1>}
      uVRegion {bIsValid=false, max=<0 0>, min=<0 0>}
     }
     insetPadding=<ltrb 1 1 1 1>
     textPadding=<ltrb 7 2.5 7 2.5>
    }
   }
   slot {class=HorizontalBoxSlot, properties {verticalAlignment=VAlign_Center, size {sizeRule=Fill, value=1}}}
  }
  txtMin_5 {
   widgetClassPath=&CommonTextBlock
   parent=HorizontalBox_94
   widget=txtMin_5
   properties {
    text=Y:
    visibility=Visible
    font {
     bForceMonospaced=false
     bMaterialIsStencil=false
     fontMaterial=null
     fontObject=&NotoSans
     letterSpacing=0
     monospacedWidth=1
     size=10
     skewAmount=0
     typefaceFontName=NotoSans-Regular
     outlineSettings {
      bApplyOutlineToDropShadows=false
      bMiteredCorners=false
      bSeparateFillAlpha=false
      outlineMaterial=null
      outlineSize=0
      outlineColor=<rgba 0 0 0 1>
     }
    }
    shadowOffset=<0 0>
   }
   slot {class=HorizontalBoxSlot, properties {verticalAlignment=VAlign_Center, padding=<ltrb 5 0 3 0>}}
  }
  spinY {
   widgetClassPath=&SpinBox
   bIsVariable=true
   parent=HorizontalBox_94
   widget=spinY
   properties {
    bAlwaysUsesDeltaSnap=true
    delta=0.01
    maxFractionalDigits=3
    maxSliderValue=5000
    maxValue=5000
    minDesiredWidth=90
    minSliderValue=1000
    foregroundColor {colorUseRule=UseColor_Specified, specifiedColor=<rgba 1 1 1 1>}
    widgetStyle {
     activeBackgroundBrush {
      bIsDynamicallyLoaded=false
      drawAs=RoundedBox
      imageType=NoImage
      mirroring=NoMirror
      resourceName=None
      resourceObject=null
      tiling=NoTile
      imageSize=<0 0>
      margin=<ltrb 0 0 0 0>
      outlineSettings {
       bUseBrushTransparency=false
       roundingType=FixedRadius
       width=1
       color {colorUseRule=UseColor_ColorTable, specifiedColor=<rgba 0 0 0 0>}
       cornerRadii=<4 4 4 4>
      }
      tintColor {colorUseRule=UseColor_Specified, specifiedColor=<rgba 0.0048 0.0048 0.0048 1>}
      uVRegion {bIsValid=false, max=<0 0>, min=<0 0>}
     }
     activeFillBrush {
      bIsDynamicallyLoaded=false
      drawAs=RoundedBox
      imageType=NoImage
      mirroring=NoMirror
      resourceName=None
      resourceObject=null
      tiling=NoTile
      imageSize=<0 0>
      margin=<ltrb 0 0 0 0>
      outlineSettings {
       bUseBrushTransparency=false
       roundingType=FixedRadius
       width=1
       color {colorUseRule=UseColor_Specified, specifiedColor=<rgba 0 0 0 0>}
       cornerRadii=<4 4 4 4>
      }
      tintColor {colorUseRule=UseColor_Specified, specifiedColor=<rgba 0 0.162 0.7454 1>}
      uVRegion {bIsValid=false, max=<0 0>, min=<0 0>}
     }
     arrowsImage {
      bIsDynamicallyLoaded=false
      drawAs=NoDrawType
      imageType=NoImage
      mirroring=NoMirror
      resourceName=None
      resourceObject=null
      tiling=NoTile
      imageSize=<0 0>
      margin=<ltrb 0 0 0 0>
      outlineSettings {
       bUseBrushTransparency=false
       roundingType=HalfHeightRadius
       width=0
       color {colorUseRule=UseColor_Specified, specifiedColor=<rgba 0 0 0 0>}
       cornerRadii=<0 0 0 0>
      }
      tintColor {colorUseRule=UseColor_Specified, specifiedColor=<rgba 1 1 1 1>}
      uVRegion {bIsValid=false, max=<0 0>, min=<0 0>}
     }
     backgroundBrush {
      bIsDynamicallyLoaded=false
      drawAs=RoundedBox
      imageType=NoImage
      mirroring=NoMirror
      resourceName=None
      resourceObject=null
      tiling=NoTile
      imageSize=<0 0>
      margin=<ltrb 0 0 0 0>
      outlineSettings {
       bUseBrushTransparency=false
       roundingType=FixedRadius
       width=1
       color {colorUseRule=UseColor_ColorTable, specifiedColor=<rgba 0 0 0 0>}
       cornerRadii=<4 4 4 4>
      }
      tintColor {colorUseRule=UseColor_Specified, specifiedColor=<rgba 0.0048 0.0048 0.0048 1>}
      uVRegion {bIsValid=false, max=<0 0>, min=<0 0>}
     }
     foregroundColor {colorUseRule=UseColor_Specified, specifiedColor=<rgba 1 1 1 1>}
     hoveredBackgroundBrush {
      bIsDynamicallyLoaded=false
      drawAs=RoundedBox
      imageType=NoImage
      mirroring=NoMirror
      resourceName=None
      resourceObject=null
      tiling=NoTile
      imageSize=<0 0>
      margin=<ltrb 0 0 0 0>
      outlineSettings {
       bUseBrushTransparency=false
       roundingType=FixedRadius
       width=1
       color {colorUseRule=UseColor_ColorTable, specifiedColor=<rgba 0 0 0 0>}
       cornerRadii=<4 4 4 4>
      }
      tintColor {colorUseRule=UseColor_Specified, specifiedColor=<rgba 0.0048 0.0048 0.0048 1>}
      uVRegion {bIsValid=false, max=<0 0>, min=<0 0>}
     }
     hoveredFillBrush {
      bIsDynamicallyLoaded=false
      drawAs=RoundedBox
      imageType=NoImage
      mirroring=NoMirror
      resourceName=None
      resourceObject=null
      tiling=NoTile
      imageSize=<0 0>
      margin=<ltrb 0 0 0 0>
      outlineSettings {
       bUseBrushTransparency=false
       roundingType=FixedRadius
       width=1
       color {colorUseRule=UseColor_Specified, specifiedColor=<rgba 0 0 0 0>}
       cornerRadii=<4 4 4 4>
      }
      tintColor {colorUseRule=UseColor_Specified, specifiedColor=<rgba 0.0953 0.0953 0.0953 1>}
      uVRegion {bIsValid=false, max=<0 0>, min=<0 0>}
     }
     inactiveFillBrush {
      bIsDynamicallyLoaded=false
      drawAs=RoundedBox
      imageType=NoImage
      mirroring=NoMirror
      resourceName=None
      resourceObject=null
      tiling=NoTile
      imageSize=<0 0>
      margin=<ltrb 0 0 0 0>
      outlineSettings {
       bUseBrushTransparency=false
       roundingType=FixedRadius
       width=1
       color {colorUseRule=UseColor_Specified, specifiedColor=<rgba 0 0 0 0>}
       cornerRadii=<4 4 4 4>
      }
      tintColor {colorUseRule=UseColor_Specified, specifiedColor=<rgba 0.0395 0.0395 0.0395 1>}
      uVRegion {bIsValid=false, max=<0 0>, min=<0 0>}
     }
     insetPadding=<ltrb 1 1 1 1>
     textPadding=<ltrb 7 2.5 7 2.5>
    }
   }
   slot {class=HorizontalBoxSlot, properties {verticalAlignment=VAlign_Center, size {sizeRule=Fill, value=1}}}
  }
  txtMin_6 {
   widgetClassPath=&CommonTextBlock
   parent=HorizontalBox_94
   widget=txtMin_6
   properties {
    text=Z:
    visibility=Visible
    font {
     bForceMonospaced=false
     bMaterialIsStencil=false
     fontMaterial=null
     fontObject=&NotoSans
     letterSpacing=0
     monospacedWidth=1
     size=10
     skewAmount=0
     typefaceFontName=NotoSans-Regular
     outlineSettings {
      bApplyOutlineToDropShadows=false
      bMiteredCorners=false
      bSeparateFillAlpha=false
      outlineMaterial=null
      outlineSize=0
      outlineColor=<rgba 0 0 0 1>
     }
    }
    shadowOffset=<0 0>
   }
   slot {class=HorizontalBoxSlot, properties {verticalAlignment=VAlign_Center, padding=<ltrb 5 0 3 0>}}
  }
  spinZ {
   widgetClassPath=&SpinBox
   bIsVariable=true
   parent=HorizontalBox_94
   widget=spinZ
   properties {
    bAlwaysUsesDeltaSnap=true
    delta=0.01
    maxFractionalDigits=3
    maxSliderValue=5000
    maxValue=5000
    minDesiredWidth=90
    minSliderValue=1000
    foregroundColor {colorUseRule=UseColor_Specified, specifiedColor=<rgba 1 1 1 1>}
    widgetStyle {
     activeBackgroundBrush {
      bIsDynamicallyLoaded=false
      drawAs=RoundedBox
      imageType=NoImage
      mirroring=NoMirror
      resourceName=None
      resourceObject=null
      tiling=NoTile
      imageSize=<0 0>
      margin=<ltrb 0 0 0 0>
      outlineSettings {
       bUseBrushTransparency=false
       roundingType=FixedRadius
       width=1
       color {colorUseRule=UseColor_ColorTable, specifiedColor=<rgba 0 0 0 0>}
       cornerRadii=<4 4 4 4>
      }
      tintColor {colorUseRule=UseColor_Specified, specifiedColor=<rgba 0.0048 0.0048 0.0048 1>}
      uVRegion {bIsValid=false, max=<0 0>, min=<0 0>}
     }
     activeFillBrush {
      bIsDynamicallyLoaded=false
      drawAs=RoundedBox
      imageType=NoImage
      mirroring=NoMirror
      resourceName=None
      resourceObject=null
      tiling=NoTile
      imageSize=<0 0>
      margin=<ltrb 0 0 0 0>
      outlineSettings {
       bUseBrushTransparency=false
       roundingType=FixedRadius
       width=1
       color {colorUseRule=UseColor_Specified, specifiedColor=<rgba 0 0 0 0>}
       cornerRadii=<4 4 4 4>
      }
      tintColor {colorUseRule=UseColor_Specified, specifiedColor=<rgba 0 0.162 0.7454 1>}
      uVRegion {bIsValid=false, max=<0 0>, min=<0 0>}
     }
     arrowsImage {
      bIsDynamicallyLoaded=false
      drawAs=NoDrawType
      imageType=NoImage
      mirroring=NoMirror
      resourceName=None
      resourceObject=null
      tiling=NoTile
      imageSize=<0 0>
      margin=<ltrb 0 0 0 0>
      outlineSettings {
       bUseBrushTransparency=false
       roundingType=HalfHeightRadius
       width=0
       color {colorUseRule=UseColor_Specified, specifiedColor=<rgba 0 0 0 0>}
       cornerRadii=<0 0 0 0>
      }
      tintColor {colorUseRule=UseColor_Specified, specifiedColor=<rgba 1 1 1 1>}
      uVRegion {bIsValid=false, max=<0 0>, min=<0 0>}
     }
     backgroundBrush {
      bIsDynamicallyLoaded=false
      drawAs=RoundedBox
      imageType=NoImage
      mirroring=NoMirror
      resourceName=None
      resourceObject=null
      tiling=NoTile
      imageSize=<0 0>
      margin=<ltrb 0 0 0 0>
      outlineSettings {
       bUseBrushTransparency=false
       roundingType=FixedRadius
       width=1
       color {colorUseRule=UseColor_ColorTable, specifiedColor=<rgba 0 0 0 0>}
       cornerRadii=<4 4 4 4>
      }
      tintColor {colorUseRule=UseColor_Specified, specifiedColor=<rgba 0.0048 0.0048 0.0048 1>}
      uVRegion {bIsValid=false, max=<0 0>, min=<0 0>}
     }
     foregroundColor {colorUseRule=UseColor_Specified, specifiedColor=<rgba 1 1 1 1>}
     hoveredBackgroundBrush {
      bIsDynamicallyLoaded=false
      drawAs=RoundedBox
      imageType=NoImage
      mirroring=NoMirror
      resourceName=None
      resourceObject=null
      tiling=NoTile
      imageSize=<0 0>
      margin=<ltrb 0 0 0 0>
      outlineSettings {
       bUseBrushTransparency=false
       roundingType=FixedRadius
       width=1
       color {colorUseRule=UseColor_ColorTable, specifiedColor=<rgba 0 0 0 0>}
       cornerRadii=<4 4 4 4>
      }
      tintColor {colorUseRule=UseColor_Specified, specifiedColor=<rgba 0.0048 0.0048 0.0048 1>}
      uVRegion {bIsValid=false, max=<0 0>, min=<0 0>}
     }
     hoveredFillBrush {
      bIsDynamicallyLoaded=false
      drawAs=RoundedBox
      imageType=NoImage
      mirroring=NoMirror
      resourceName=None
      resourceObject=null
      tiling=NoTile
      imageSize=<0 0>
      margin=<ltrb 0 0 0 0>
      outlineSettings {
       bUseBrushTransparency=false
       roundingType=FixedRadius
       width=1
       color {colorUseRule=UseColor_Specified, specifiedColor=<rgba 0 0 0 0>}
       cornerRadii=<4 4 4 4>
      }
      tintColor {colorUseRule=UseColor_Specified, specifiedColor=<rgba 0.0953 0.0953 0.0953 1>}
      uVRegion {bIsValid=false, max=<0 0>, min=<0 0>}
     }
     inactiveFillBrush {
      bIsDynamicallyLoaded=false
      drawAs=RoundedBox
      imageType=NoImage
      mirroring=NoMirror
      resourceName=None
      resourceObject=null
      tiling=NoTile
      imageSize=<0 0>
      margin=<ltrb 0 0 0 0>
      outlineSettings {
       bUseBrushTransparency=false
       roundingType=FixedRadius
       width=1
       color {colorUseRule=UseColor_Specified, specifiedColor=<rgba 0 0 0 0>}
       cornerRadii=<4 4 4 4>
      }
      tintColor {colorUseRule=UseColor_Specified, specifiedColor=<rgba 0.0395 0.0395 0.0395 1>}
      uVRegion {bIsValid=false, max=<0 0>, min=<0 0>}
     }
     insetPadding=<ltrb 1 1 1 1>
     textPadding=<ltrb 7 2.5 7 2.5>
    }
   }
   slot {class=HorizontalBoxSlot, properties {verticalAlignment=VAlign_Center, size {sizeRule=Fill, value=1}}}
  }
  btnPasteLoc {
   widgetClassPath=&TA_Generic01_Button
   parent=HorizontalBox_94
   widget=btnPasteLoc
   properties {label=Paste Location, toolTipText=Paste a previously copied Location, buttonColor_Default=<rgba 0.2782 0.6979 0.0962 1>}
   slot {class=HorizontalBoxSlot, properties {horizontalAlignment=HAlign_Right, verticalAlignment=VAlign_Center, padding=<ltrb 5 0 0 0>}}
  }
  btnClearLoc {
   widgetClassPath=&TA_Generic01_Button
   parent=HorizontalBox_94
   widget=btnClearLoc
   properties {label=Clear, toolTipText=Clear Location only, buttonColor_Default=<rgba 1 0.1042 0.1042 1>}
   slot {class=HorizontalBoxSlot, properties {horizontalAlignment=HAlign_Right, verticalAlignment=VAlign_Center, padding=<ltrb 5 0 0 0>}}
  }
  HorizontalBox {
   widgetClassPath=&HorizontalBox
   parent=VerticalBox_2
   widget=HorizontalBox
   slot {class=VerticalBoxSlot, properties {padding=<ltrb 0 5 0 5>}}
  }
  textField_1 {
   widgetClassPath=&CommonTextBlock
   parent=HorizontalBox
   widget=textField_1
   properties {
    justification=Right
    minDesiredWidth=150
    text=Rotation Offset
    visibility=Visible
    font {
     bForceMonospaced=false
     bMaterialIsStencil=false
     fontMaterial=null
     fontObject=&NotoSans
     letterSpacing=0
     monospacedWidth=1
     size=12
     skewAmount=0
     typefaceFontName=NotoSans-Regular
     outlineSettings {
      bApplyOutlineToDropShadows=false
      bMiteredCorners=false
      bSeparateFillAlpha=false
      outlineMaterial=null
      outlineSize=0
      outlineColor=<rgba 0 0 0 1>
     }
    }
    shadowOffset=<0 0>
   }
   slot {class=HorizontalBoxSlot, properties {verticalAlignment=VAlign_Center}}
  }
  txtMin_1 {
   widgetClassPath=&CommonTextBlock
   parent=HorizontalBox
   widget=txtMin_1
   properties {
    text=R:
    visibility=Visible
    font {
     bForceMonospaced=false
     bMaterialIsStencil=false
     fontMaterial=null
     fontObject=&NotoSans
     letterSpacing=0
     monospacedWidth=1
     size=10
     skewAmount=0
     typefaceFontName=NotoSans-Regular
     outlineSettings {
      bApplyOutlineToDropShadows=false
      bMiteredCorners=false
      bSeparateFillAlpha=false
      outlineMaterial=null
      outlineSize=0
      outlineColor=<rgba 0 0 0 1>
     }
    }
    shadowOffset=<0 0>
   }
   slot {class=HorizontalBoxSlot, properties {verticalAlignment=VAlign_Center, padding=<ltrb 5 0 3 0>}}
  }
  spinRoll {
   widgetClassPath=&SpinBox
   bIsVariable=true
   parent=HorizontalBox
   widget=spinRoll
   properties {
    bAlwaysUsesDeltaSnap=true
    delta=0.01
    maxFractionalDigits=3
    maxSliderValue=5000
    maxValue=5000
    minDesiredWidth=90
    minSliderValue=1000
    foregroundColor {colorUseRule=UseColor_Specified, specifiedColor=<rgba 1 1 1 1>}
    widgetStyle {
     activeBackgroundBrush {
      bIsDynamicallyLoaded=false
      drawAs=RoundedBox
      imageType=NoImage
      mirroring=NoMirror
      resourceName=None
      resourceObject=null
      tiling=NoTile
      imageSize=<0 0>
      margin=<ltrb 0 0 0 0>
      outlineSettings {
       bUseBrushTransparency=false
       roundingType=FixedRadius
       width=1
       color {colorUseRule=UseColor_ColorTable, specifiedColor=<rgba 0 0 0 0>}
       cornerRadii=<4 4 4 4>
      }
      tintColor {colorUseRule=UseColor_Specified, specifiedColor=<rgba 0.0048 0.0048 0.0048 1>}
      uVRegion {bIsValid=false, max=<0 0>, min=<0 0>}
     }
     activeFillBrush {
      bIsDynamicallyLoaded=false
      drawAs=RoundedBox
      imageType=NoImage
      mirroring=NoMirror
      resourceName=None
      resourceObject=null
      tiling=NoTile
      imageSize=<0 0>
      margin=<ltrb 0 0 0 0>
      outlineSettings {
       bUseBrushTransparency=false
       roundingType=FixedRadius
       width=1
       color {colorUseRule=UseColor_Specified, specifiedColor=<rgba 0 0 0 0>}
       cornerRadii=<4 4 4 4>
      }
      tintColor {colorUseRule=UseColor_Specified, specifiedColor=<rgba 0 0.162 0.7454 1>}
      uVRegion {bIsValid=false, max=<0 0>, min=<0 0>}
     }
     arrowsImage {
      bIsDynamicallyLoaded=false
      drawAs=NoDrawType
      imageType=NoImage
      mirroring=NoMirror
      resourceName=None
      resourceObject=null
      tiling=NoTile
      imageSize=<0 0>
      margin=<ltrb 0 0 0 0>
      outlineSettings {
       bUseBrushTransparency=false
       roundingType=HalfHeightRadius
       width=0
       color {colorUseRule=UseColor_Specified, specifiedColor=<rgba 0 0 0 0>}
       cornerRadii=<0 0 0 0>
      }
      tintColor {colorUseRule=UseColor_Specified, specifiedColor=<rgba 1 1 1 1>}
      uVRegion {bIsValid=false, max=<0 0>, min=<0 0>}
     }
     backgroundBrush {
      bIsDynamicallyLoaded=false
      drawAs=RoundedBox
      imageType=NoImage
      mirroring=NoMirror
      resourceName=None
      resourceObject=null
      tiling=NoTile
      imageSize=<0 0>
      margin=<ltrb 0 0 0 0>
      outlineSettings {
       bUseBrushTransparency=false
       roundingType=FixedRadius
       width=1
       color {colorUseRule=UseColor_ColorTable, specifiedColor=<rgba 0 0 0 0>}
       cornerRadii=<4 4 4 4>
      }
      tintColor {colorUseRule=UseColor_Specified, specifiedColor=<rgba 0.0048 0.0048 0.0048 1>}
      uVRegion {bIsValid=false, max=<0 0>, min=<0 0>}
     }
     foregroundColor {colorUseRule=UseColor_Specified, specifiedColor=<rgba 1 1 1 1>}
     hoveredBackgroundBrush {
      bIsDynamicallyLoaded=false
      drawAs=RoundedBox
      imageType=NoImage
      mirroring=NoMirror
      resourceName=None
      resourceObject=null
      tiling=NoTile
      imageSize=<0 0>
      margin=<ltrb 0 0 0 0>
      outlineSettings {
       bUseBrushTransparency=false
       roundingType=FixedRadius
       width=1
       color {colorUseRule=UseColor_ColorTable, specifiedColor=<rgba 0 0 0 0>}
       cornerRadii=<4 4 4 4>
      }
      tintColor {colorUseRule=UseColor_Specified, specifiedColor=<rgba 0.0048 0.0048 0.0048 1>}
      uVRegion {bIsValid=false, max=<0 0>, min=<0 0>}
     }
     hoveredFillBrush {
      bIsDynamicallyLoaded=false
      drawAs=RoundedBox
      imageType=NoImage
      mirroring=NoMirror
      resourceName=None
      resourceObject=null
      tiling=NoTile
      imageSize=<0 0>
      margin=<ltrb 0 0 0 0>
      outlineSettings {
       bUseBrushTransparency=false
       roundingType=FixedRadius
       width=1
       color {colorUseRule=UseColor_Specified, specifiedColor=<rgba 0 0 0 0>}
       cornerRadii=<4 4 4 4>
      }
      tintColor {colorUseRule=UseColor_Specified, specifiedColor=<rgba 0.0953 0.0953 0.0953 1>}
      uVRegion {bIsValid=false, max=<0 0>, min=<0 0>}
     }
     inactiveFillBrush {
      bIsDynamicallyLoaded=false
      drawAs=RoundedBox
      imageType=NoImage
      mirroring=NoMirror
      resourceName=None
      resourceObject=null
      tiling=NoTile
      imageSize=<0 0>
      margin=<ltrb 0 0 0 0>
      outlineSettings {
       bUseBrushTransparency=false
       roundingType=FixedRadius
       width=1
       color {colorUseRule=UseColor_Specified, specifiedColor=<rgba 0 0 0 0>}
       cornerRadii=<4 4 4 4>
      }
      tintColor {colorUseRule=UseColor_Specified, specifiedColor=<rgba 0.0395 0.0395 0.0395 1>}
      uVRegion {bIsValid=false, max=<0 0>, min=<0 0>}
     }
     insetPadding=<ltrb 1 1 1 1>
     textPadding=<ltrb 7 2.5 7 2.5>
    }
   }
   slot {class=HorizontalBoxSlot, properties {verticalAlignment=VAlign_Center, size {sizeRule=Fill, value=1}}}
  }
  txtMin_2 {
   widgetClassPath=&CommonTextBlock
   parent=HorizontalBox
   widget=txtMin_2
   properties {
    text=P:
    visibility=Visible
    font {
     bForceMonospaced=false
     bMaterialIsStencil=false
     fontMaterial=null
     fontObject=&NotoSans
     letterSpacing=0
     monospacedWidth=1
     size=10
     skewAmount=0
     typefaceFontName=NotoSans-Regular
     outlineSettings {
      bApplyOutlineToDropShadows=false
      bMiteredCorners=false
      bSeparateFillAlpha=false
      outlineMaterial=null
      outlineSize=0
      outlineColor=<rgba 0 0 0 1>
     }
    }
    shadowOffset=<0 0>
   }
   slot {class=HorizontalBoxSlot, properties {verticalAlignment=VAlign_Center, padding=<ltrb 5 0 3 0>}}
  }
  spinPitch {
   widgetClassPath=&SpinBox
   bIsVariable=true
   parent=HorizontalBox
   widget=spinPitch
   properties {
    bAlwaysUsesDeltaSnap=true
    delta=0.01
    maxFractionalDigits=3
    maxSliderValue=5000
    maxValue=5000
    minDesiredWidth=90
    minSliderValue=1000
    foregroundColor {colorUseRule=UseColor_Specified, specifiedColor=<rgba 1 1 1 1>}
    widgetStyle {
     activeBackgroundBrush {
      bIsDynamicallyLoaded=false
      drawAs=RoundedBox
      imageType=NoImage
      mirroring=NoMirror
      resourceName=None
      resourceObject=null
      tiling=NoTile
      imageSize=<0 0>
      margin=<ltrb 0 0 0 0>
      outlineSettings {
       bUseBrushTransparency=false
       roundingType=FixedRadius
       width=1
       color {colorUseRule=UseColor_ColorTable, specifiedColor=<rgba 0 0 0 0>}
       cornerRadii=<4 4 4 4>
      }
      tintColor {colorUseRule=UseColor_Specified, specifiedColor=<rgba 0.0048 0.0048 0.0048 1>}
      uVRegion {bIsValid=false, max=<0 0>, min=<0 0>}
     }
     activeFillBrush {
      bIsDynamicallyLoaded=false
      drawAs=RoundedBox
      imageType=NoImage
      mirroring=NoMirror
      resourceName=None
      resourceObject=null
      tiling=NoTile
      imageSize=<0 0>
      margin=<ltrb 0 0 0 0>
      outlineSettings {
       bUseBrushTransparency=false
       roundingType=FixedRadius
       width=1
       color {colorUseRule=UseColor_Specified, specifiedColor=<rgba 0 0 0 0>}
       cornerRadii=<4 4 4 4>
      }
      tintColor {colorUseRule=UseColor_Specified, specifiedColor=<rgba 0 0.162 0.7454 1>}
      uVRegion {bIsValid=false, max=<0 0>, min=<0 0>}
     }
     arrowsImage {
      bIsDynamicallyLoaded=false
      drawAs=NoDrawType
      imageType=NoImage
      mirroring=NoMirror
      resourceName=None
      resourceObject=null
      tiling=NoTile
      imageSize=<0 0>
      margin=<ltrb 0 0 0 0>
      outlineSettings {
       bUseBrushTransparency=false
       roundingType=HalfHeightRadius
       width=0
       color {colorUseRule=UseColor_Specified, specifiedColor=<rgba 0 0 0 0>}
       cornerRadii=<0 0 0 0>
      }
      tintColor {colorUseRule=UseColor_Specified, specifiedColor=<rgba 1 1 1 1>}
      uVRegion {bIsValid=false, max=<0 0>, min=<0 0>}
     }
     backgroundBrush {
      bIsDynamicallyLoaded=false
      drawAs=RoundedBox
      imageType=NoImage
      mirroring=NoMirror
      resourceName=None
      resourceObject=null
      tiling=NoTile
      imageSize=<0 0>
      margin=<ltrb 0 0 0 0>
      outlineSettings {
       bUseBrushTransparency=false
       roundingType=FixedRadius
       width=1
       color {colorUseRule=UseColor_ColorTable, specifiedColor=<rgba 0 0 0 0>}
       cornerRadii=<4 4 4 4>
      }
      tintColor {colorUseRule=UseColor_Specified, specifiedColor=<rgba 0.0048 0.0048 0.0048 1>}
      uVRegion {bIsValid=false, max=<0 0>, min=<0 0>}
     }
     foregroundColor {colorUseRule=UseColor_Specified, specifiedColor=<rgba 1 1 1 1>}
     hoveredBackgroundBrush {
      bIsDynamicallyLoaded=false
      drawAs=RoundedBox
      imageType=NoImage
      mirroring=NoMirror
      resourceName=None
      resourceObject=null
      tiling=NoTile
      imageSize=<0 0>
      margin=<ltrb 0 0 0 0>
      outlineSettings {
       bUseBrushTransparency=false
       roundingType=FixedRadius
       width=1
       color {colorUseRule=UseColor_ColorTable, specifiedColor=<rgba 0 0 0 0>}
       cornerRadii=<4 4 4 4>
      }
      tintColor {colorUseRule=UseColor_Specified, specifiedColor=<rgba 0.0048 0.0048 0.0048 1>}
      uVRegion {bIsValid=false, max=<0 0>, min=<0 0>}
     }
     hoveredFillBrush {
      bIsDynamicallyLoaded=false
      drawAs=RoundedBox
      imageType=NoImage
      mirroring=NoMirror
      resourceName=None
      resourceObject=null
      tiling=NoTile
      imageSize=<0 0>
      margin=<ltrb 0 0 0 0>
      outlineSettings {
       bUseBrushTransparency=false
       roundingType=FixedRadius
       width=1
       color {colorUseRule=UseColor_Specified, specifiedColor=<rgba 0 0 0 0>}
       cornerRadii=<4 4 4 4>
      }
      tintColor {colorUseRule=UseColor_Specified, specifiedColor=<rgba 0.0953 0.0953 0.0953 1>}
      uVRegion {bIsValid=false, max=<0 0>, min=<0 0>}
     }
     inactiveFillBrush {
      bIsDynamicallyLoaded=false
      drawAs=RoundedBox
      imageType=NoImage
      mirroring=NoMirror
      resourceName=None
      resourceObject=null
      tiling=NoTile
      imageSize=<0 0>
      margin=<ltrb 0 0 0 0>
      outlineSettings {
       bUseBrushTransparency=false
       roundingType=FixedRadius
       width=1
       color {colorUseRule=UseColor_Specified, specifiedColor=<rgba 0 0 0 0>}
       cornerRadii=<4 4 4 4>
      }
      tintColor {colorUseRule=UseColor_Specified, specifiedColor=<rgba 0.0395 0.0395 0.0395 1>}
      uVRegion {bIsValid=false, max=<0 0>, min=<0 0>}
     }
     insetPadding=<ltrb 1 1 1 1>
     textPadding=<ltrb 7 2.5 7 2.5>
    }
   }
   slot {class=HorizontalBoxSlot, properties {verticalAlignment=VAlign_Center, size {sizeRule=Fill, value=1}}}
  }
  txtMin_3 {
   widgetClassPath=&CommonTextBlock
   parent=HorizontalBox
   widget=txtMin_3
   properties {
    text=Y:
    visibility=Visible
    font {
     bForceMonospaced=false
     bMaterialIsStencil=false
     fontMaterial=null
     fontObject=&NotoSans
     letterSpacing=0
     monospacedWidth=1
     size=10
     skewAmount=0
     typefaceFontName=NotoSans-Regular
     outlineSettings {
      bApplyOutlineToDropShadows=false
      bMiteredCorners=false
      bSeparateFillAlpha=false
      outlineMaterial=null
      outlineSize=0
      outlineColor=<rgba 0 0 0 1>
     }
    }
    shadowOffset=<0 0>
   }
   slot {class=HorizontalBoxSlot, properties {verticalAlignment=VAlign_Center, padding=<ltrb 5 0 3 0>}}
  }
  spinYaw {
   widgetClassPath=&SpinBox
   bIsVariable=true
   parent=HorizontalBox
   widget=spinYaw
   properties {
    bAlwaysUsesDeltaSnap=true
    delta=0.01
    maxFractionalDigits=3
    maxSliderValue=5000
    maxValue=5000
    minDesiredWidth=90
    minSliderValue=1000
    foregroundColor {colorUseRule=UseColor_Specified, specifiedColor=<rgba 1 1 1 1>}
    widgetStyle {
     activeBackgroundBrush {
      bIsDynamicallyLoaded=false
      drawAs=RoundedBox
      imageType=NoImage
      mirroring=NoMirror
      resourceName=None
      resourceObject=null
      tiling=NoTile
      imageSize=<0 0>
      margin=<ltrb 0 0 0 0>
      outlineSettings {
       bUseBrushTransparency=false
       roundingType=FixedRadius
       width=1
       color {colorUseRule=UseColor_ColorTable, specifiedColor=<rgba 0 0 0 0>}
       cornerRadii=<4 4 4 4>
      }
      tintColor {colorUseRule=UseColor_Specified, specifiedColor=<rgba 0.0048 0.0048 0.0048 1>}
      uVRegion {bIsValid=false, max=<0 0>, min=<0 0>}
     }
     activeFillBrush {
      bIsDynamicallyLoaded=false
      drawAs=RoundedBox
      imageType=NoImage
      mirroring=NoMirror
      resourceName=None
      resourceObject=null
      tiling=NoTile
      imageSize=<0 0>
      margin=<ltrb 0 0 0 0>
      outlineSettings {
       bUseBrushTransparency=false
       roundingType=FixedRadius
       width=1
       color {colorUseRule=UseColor_Specified, specifiedColor=<rgba 0 0 0 0>}
       cornerRadii=<4 4 4 4>
      }
      tintColor {colorUseRule=UseColor_Specified, specifiedColor=<rgba 0 0.162 0.7454 1>}
      uVRegion {bIsValid=false, max=<0 0>, min=<0 0>}
     }
     arrowsImage {
      bIsDynamicallyLoaded=false
      drawAs=NoDrawType
      imageType=NoImage
      mirroring=NoMirror
      resourceName=None
      resourceObject=null
      tiling=NoTile
      imageSize=<0 0>
      margin=<ltrb 0 0 0 0>
      outlineSettings {
       bUseBrushTransparency=false
       roundingType=HalfHeightRadius
       width=0
       color {colorUseRule=UseColor_Specified, specifiedColor=<rgba 0 0 0 0>}
       cornerRadii=<0 0 0 0>
      }
      tintColor {colorUseRule=UseColor_Specified, specifiedColor=<rgba 1 1 1 1>}
      uVRegion {bIsValid=false, max=<0 0>, min=<0 0>}
     }
     backgroundBrush {
      bIsDynamicallyLoaded=false
      drawAs=RoundedBox
      imageType=NoImage
      mirroring=NoMirror
      resourceName=None
      resourceObject=null
      tiling=NoTile
      imageSize=<0 0>
      margin=<ltrb 0 0 0 0>
      outlineSettings {
       bUseBrushTransparency=false
       roundingType=FixedRadius
       width=1
       color {colorUseRule=UseColor_ColorTable, specifiedColor=<rgba 0 0 0 0>}
       cornerRadii=<4 4 4 4>
      }
      tintColor {colorUseRule=UseColor_Specified, specifiedColor=<rgba 0.0048 0.0048 0.0048 1>}
      uVRegion {bIsValid=false, max=<0 0>, min=<0 0>}
     }
     foregroundColor {colorUseRule=UseColor_Specified, specifiedColor=<rgba 1 1 1 1>}
     hoveredBackgroundBrush {
      bIsDynamicallyLoaded=false
      drawAs=RoundedBox
      imageType=NoImage
      mirroring=NoMirror
      resourceName=None
      resourceObject=null
      tiling=NoTile
      imageSize=<0 0>
      margin=<ltrb 0 0 0 0>
      outlineSettings {
       bUseBrushTransparency=false
       roundingType=FixedRadius
       width=1
       color {colorUseRule=UseColor_ColorTable, specifiedColor=<rgba 0 0 0 0>}
       cornerRadii=<4 4 4 4>
      }
      tintColor {colorUseRule=UseColor_Specified, specifiedColor=<rgba 0.0048 0.0048 0.0048 1>}
      uVRegion {bIsValid=false, max=<0 0>, min=<0 0>}
     }
     hoveredFillBrush {
      bIsDynamicallyLoaded=false
      drawAs=RoundedBox
      imageType=NoImage
      mirroring=NoMirror
      resourceName=None
      resourceObject=null
      tiling=NoTile
      imageSize=<0 0>
      margin=<ltrb 0 0 0 0>
      outlineSettings {
       bUseBrushTransparency=false
       roundingType=FixedRadius
       width=1
       color {colorUseRule=UseColor_Specified, specifiedColor=<rgba 0 0 0 0>}
       cornerRadii=<4 4 4 4>
      }
      tintColor {colorUseRule=UseColor_Specified, specifiedColor=<rgba 0.0953 0.0953 0.0953 1>}
      uVRegion {bIsValid=false, max=<0 0>, min=<0 0>}
     }
     inactiveFillBrush {
      bIsDynamicallyLoaded=false
      drawAs=RoundedBox
      imageType=NoImage
      mirroring=NoMirror
      resourceName=None
      resourceObject=null
      tiling=NoTile
      imageSize=<0 0>
      margin=<ltrb 0 0 0 0>
      outlineSettings {
       bUseBrushTransparency=false
       roundingType=FixedRadius
       width=1
       color {colorUseRule=UseColor_Specified, specifiedColor=<rgba 0 0 0 0>}
       cornerRadii=<4 4 4 4>
      }
      tintColor {colorUseRule=UseColor_Specified, specifiedColor=<rgba 0.0395 0.0395 0.0395 1>}
      uVRegion {bIsValid=false, max=<0 0>, min=<0 0>}
     }
     insetPadding=<ltrb 1 1 1 1>
     textPadding=<ltrb 7 2.5 7 2.5>
    }
   }
   slot {class=HorizontalBoxSlot, properties {verticalAlignment=VAlign_Center, size {sizeRule=Fill, value=1}}}
  }
  btnPasteRot {
   widgetClassPath=&TA_Generic01_Button
   parent=HorizontalBox
   widget=btnPasteRot
   properties {label=Paste Rotation, toolTipText=Paste a previously copied Rotation, buttonColor_Default=<rgba 0.2782 0.6979 0.0962 1>}
   slot {class=HorizontalBoxSlot, properties {horizontalAlignment=HAlign_Right, verticalAlignment=VAlign_Center, padding=<ltrb 5 0 0 0>}}
  }
  btnClearRot {
   widgetClassPath=&TA_Generic01_Button
   parent=HorizontalBox
   widget=btnClearRot
   properties {label=Clear, toolTipText=Clear Rotation only, buttonColor_Default=<rgba 1 0.1042 0.1042 1>}
   slot {class=HorizontalBoxSlot, properties {horizontalAlignment=HAlign_Right, verticalAlignment=VAlign_Center, padding=<ltrb 5 0 0 0>}}
  }
 ]
}
