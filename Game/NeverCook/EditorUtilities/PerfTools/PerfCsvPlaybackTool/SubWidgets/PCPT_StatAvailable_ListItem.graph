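# Editor list-item widget for the PerfCsvPlaybackTool that shows one available stat with a checkbox for including or excluding it.
# Built on TA_BaseEditorButtonWidget, it sets the checkbox from the first bool in its TA_GenericListItemData.
# User toggles, and Set Included when requested, fire onInclusionChanged with the widget, the data's display string and the checked state.
# Label color follows the checkbox, while an edge image's color follows selection.
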WidgetBlueprint /Game/NeverCook/EditorUtilities/PerfTools/PerfCsvPlaybackTool/SubWidgets/PCPT_StatAvailable_ListItem
graphs [
 EventGraph {nodes [
   "On Check State Changed (CheckBox_315)"!() -> (OutputDelegate, then, bIsChecked:bool) {is=title K2Node_ComponentBoundEvent}
   > throwInclusionChangedEvent()
   > validateLabelColor!(execute, self:TA Base Editor Button Widget Object Reference) -> (then, handled:bool)
   n1 = Event updateAfterDataChange!() -> (OutputDelegate, then, currentListItemData:Object Reference) {is=title K2Node_Event}
   > Parent: updateAfterDataChange!(execute, currentListItemData:Object Reference=@n1.currentListItemData) -> (then) {is=title K2Node_CallParentFunction}
   > n2 = Cast To TA_GenericListItemData!(execute, Object=@n1.currentListItemData) -> (then, CastFailed, AsTA Generic List Item Data:TA Generic List Item Data Object Reference) {is=title K2Node_DynamicCast}
   > SetIsChecked!(
    execute
    self:Check Box Object Reference=Get CheckBox_315() -> (CheckBox_315:Check Box Object Reference)
    InIsChecked:bool="Get (a copy)"(Array:Array of Booleans=Get bools!(self:TA Generic List Item Data Object Reference=@n2) -> (bools:Array of Booleans)) -> (Output:bool)
   ) -> (then)
  ]}
 Set Included {nodes [
   n1 = Set Included!() -> (then, isIncluded:bool=~, throwEvent:bool=~) {is=title K2Node_FunctionEntry}
   > n2 = Sequence!(execute) -> (then_0, then_1, then_2:Exec)
   > SetIsChecked(self=Get CheckBox_315(), InIsChecked=@n1)
   Branch(execute=@n2.then_1, Condition=Get throwEvent() -> (throwEvent:bool))
   > throwInclusionChangedEvent()
   validateLabelColor(execute=@n2.then_2)
  ]}
 getBackgroundWidget {nodes [
   getBackgroundWidget() {is=title K2Node_FunctionEntry}
   > Return Node!(execute, backgroundWidget:Widget Object Reference=Get borderMain() -> (borderMain:Border Object Reference))
  ]}
 getIsIncluded {nodes [
   getIsIncluded() {is=title K2Node_FunctionEntry}
   > Return Node!(execute, isIncluded:bool=IsChecked(self=Get CheckBox_315()))
  ]}
 getLabelWidget {nodes [
   getLabelWidget() {is=title K2Node_FunctionEntry}
   > Return Node!(execute, labelWidget:Widget Object Reference=Get txtMain() -> (txtMain:Text Object Reference))
  ]}
 onInclusionChanged {nodes [onInclusionChanged!() -> (then, widget:PCPT Stat Available List Item Object Reference, stat:String, included:bool=~) {is=title K2Node_FunctionEntry}]}
 setTextAlignment {nodes [
   n1 = setTextAlignment!() -> (then, alignment:EHorizontalAlignment Enum=HAlign_Fill) {is=title K2Node_FunctionEntry}
   > n2 = Cast To ScaleBoxSlot!(execute, Object=Get Slot!(self:Widget Object Reference=Get txtMain()) -> (Slot:Panel Slot Object Reference)) -> (then, CastFailed, AsScale Box Slot:Scale Box Slot Object Reference) {is=title K2Node_DynamicCast}
   > SetHorizontalAlignment(self:Scale Box Slot Object Reference=@n2, InHorizontalAlignment=@n1)
  ]}
 throwInclusionChangedEvent {nodes [
   throwInclusionChangedEvent() {is=title K2Node_FunctionEntry}
   > n1 = Cast To TA_GenericListItemData(Object=Get currentListItemData() -> (currentListItemData:Object Reference))
   > Call On Inclusion Changed!(
    execute:Exec
    self:PCPT Stat Available List Item Object Reference
    widget:PCPT Stat Available List Item Object Reference=Self-Reference()
    stat:String=Get displayString!(self:TA Generic List Item Data Object Reference=@n1) -> (displayString:String)
    included:bool=IsChecked(self=Get CheckBox_315())
   ) -> (then:Exec) {is=title K2Node_CallDelegate}
  ]}
 validateBackgroundColor {nodes [
   validateBackgroundColor() {is=title K2Node_FunctionEntry}
   > SetColorAndOpacity(
    self:Image Object Reference=Get imageEdge() -> (imageEdge:Image Object Reference)
    InColorAndOpacity:Linear Color Structure=SelectColor(A=(R=0.0284,G=0.3763,B=1,A=1), B=(R=0.8123,G=0.8217,B=0.8385,A=1), bPickA=getSelected!() -> (isSelected:bool))
   )
   > n1 = Parent: validateBackgroundColor!(execute) -> (then, handled:bool) {is=title K2Node_CallParentFunction}
   > Return Node!(execute, handled:bool=@n1)
  ]}
 validateLabelColor {nodes [
   validateLabelColor() {is=title K2Node_FunctionEntry}
   n1 = IsChecked(self=Get CheckBox_315())
   > SetColorAndOpacity(
    self:Text Object Reference=Get txtMain()
    InColorAndOpacity:Slate Color Structure=Make Slate Color(SpecifiedColor=SelectColor(A=(R=0.0719,G=0.3039,B=1,A=1), B=(R=0.9219,G=0.9219,B=0.9219,A=1), bPickA=@n1))
   )
   > Return Node(handled=true)
  ]}
]
variables [onInclusionChanged {}]
widgets {
 info {
  inheritedWidgetCount=0
  parentClass=/Game/Blueprints/Blutilities/ForTechArt/Widgets/Buttons/TA_BaseEditorButtonWidget.TA_BaseEditorButtonWidget_C
  rootWidgetClass=&Overlay
  widgetCount=8
 }
 widgets [
  Overlay_7 {widgetClassPath=&Overlay, widget=Overlay_7}
  VerticalBox_0 {
   widgetClassPath=&VerticalBox
   parent=Overlay_7
   widget=VerticalBox_0
   slot {class=OverlaySlot, properties {horizontalAlignment=HAlign_Fill}}
  }
  borderMain {
   widgetClassPath=&Border
   bIsVariable=true
   parent=VerticalBox_0
   widget=borderMain
   properties {padding=<ltrb 3 3 3 3>}
   slot {class=VerticalBoxSlot}
  }
  HorizontalBox_0 {
   widgetClassPath=&HorizontalBox
   parent=borderMain
   widget=HorizontalBox_0
   slot {class=BorderSlot, properties {padding=<ltrb 3 3 3 3>}}
  }
  CheckBox_315 {widgetClassPath=/Script/UMG.CheckBox, bIsVariable=true, parent=HorizontalBox_0, widget=CheckBox_315, slot {class=HorizontalBoxSlot}}
  ScaleBox_0 {
   widgetClassPath=&ScaleBox
   parent=HorizontalBox_0
   widget=ScaleBox_0
   properties {clipping=ClipToBounds, stretchDirection=DownOnly}
   slot {class=HorizontalBoxSlot, properties {padding=<ltrb 5 0 0 0>, size {sizeRule=Fill, value=1}}}
  }
  txtMain {
   widgetClassPath=&TextBlock
   bIsVariable=true
   parent=ScaleBox_0
   widget=txtMain
   properties {
    clipping=ClipToBounds
    justification=Center
    text=Text Block
    colorAndOpacity {colorUseRule=UseColor_Specified, specifiedColor=<rgba 0.0104 0.0104 0.0104 1>}
    font {
     bForceMonospaced=false
     bMaterialIsStencil=false
     fontMaterial=null
     fontObject=&Roboto
     letterSpacing=0
     monospacedWidth=1
     size=10
     skewAmount=0
     typefaceFontName=Bold
     outlineSettings {
      bApplyOutlineToDropShadows=false
      bMiteredCorners=false
      bSeparateFillAlpha=false
      outlineMaterial=null
      outlineSize=0
      outlineColor=<rgba 0 0 0 1>
     }
    }
    shadowColorAndOpacity=<rgba 0 0 0 0.2>
   }
   slot {class=ScaleBoxSlot, properties {horizontalAlignment=HAlign_Fill}}
  }
  imageEdge {
   widgetClassPath=&Image
   bIsVariable=true
   parent=VerticalBox_0
   widget=imageEdge
   properties {
    brush {
     bIsDynamicallyLoaded=false
     drawAs=Image
     imageType=NoImage
     mirroring=NoMirror
     resourceName=None
     resourceObject=null
     tiling=NoTile
     imageSize=<32 2>
     margin=<ltrb 0 0 0 0>
     outlineSettings {
      bUseBrushTransparency=false
      roundingType=HalfHeightRadius
      width=0
      color {colorUseRule=UseColor_Specified, specifiedColor=<rgba 0 0 0 0>}
      cornerRadii=<0 0 0 0>
     }
     tintColor {colorUseRule=UseColor_Specified, specifiedColor=<rgba 1 1 1 1>}
     uVRegion {bIsValid=false, max=<0 0>, min=<0 0>}
    }
    colorAndOpacity=<rgba 0.2031 0.2031 0.2031 1>
   }
   slot {class=VerticalBoxSlot}
  }
 ]
}
