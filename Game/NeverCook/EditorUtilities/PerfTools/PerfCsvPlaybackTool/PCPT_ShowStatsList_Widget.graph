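# Editor utility overlay for the Perf CSV Playback tool that selects which stats to display.
# setAvailableStats builds list rows, included by saved hidden categories or, when none exist, a default set.
# Row clicks, text filter and Toggle All (inverting the first row's state) drive the list, relaying changes via onStatInclusionChanged.
# requestHiddenCategories returns excluded labels, whereas setHiddenCategories marks matching labels as included, without negation.
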EditorUtilityWidgetBlueprint /Game/NeverCook/EditorUtilities/PerfTools/PerfCsvPlaybackTool/PCPT_ShowStatsList_Widget
graphs [
 EventGraph {nodes [
   n1 = "On Item Left Clicked (listStatAvailable)"!() -> (OutputDelegate, then, isShiftDown:bool, isCtrlDown:bool, data:Object Reference, item:User Widget Object Reference) {is=title K2Node_ComponentBoundEvent}
   > n2 = Cast To TA_GenericListItemData!(execute, Object=@n1.data) -> (then, CastFailed, AsTA Generic List Item Data:TA Generic List Item Data Object Reference) {is=title K2Node_DynamicCast}
   > n3 = getListItems!(execute, self:TA Generic 01 List Object Reference=Get listStatAvailable() -> (listStatAvailable:TA Generic 01 List Object Reference)) -> (then, items:Array of User Widget Object References)
   > n4 = For Each Loop with Break!(Exec:Exec, Array:Array of User Widget Object References=@n3, Break:Exec=@n6.then) -> (LoopBody:Exec, Array Element:User Widget Object Reference, Array Index:int, Completed:Exec)
   > n5 = Cast To PCPT_StatAvailable_ListItem!(execute, Object=@n4) -> (then, CastFailed, AsPCPT Stat Available List Item:PCPT Stat Available List Item Object Reference) {is=title K2Node_DynamicCast}
   > Branch(Condition="Equal Exactly (String)"(
     A=Get label!(self:TA Base Editor Button Widget Object Reference=@n5) -> (label:String)
     B=Get displayString!(self:TA Generic List Item Data Object Reference=@n2) -> (displayString:String)
    ))
   > n6 = Set Included!(
    execute
    self:PCPT Stat Available List Item Object Reference=@n5
    isIncluded:bool=NOT Boolean(A=getIsIncluded!(self:PCPT Stat Available List Item Object Reference=@n5) -> (isIncluded:bool))
    throwEvent:bool=true
   ) -> (then) {is=title K2Node_CallFunction}
   "On Text Changed (txtBoxFilter)"!() -> (OutputDelegate, then, Text:"Text (by ref)") {is=title K2Node_ComponentBoundEvent}
   > n7 = getListItems(self=Get listStatAvailable())
   > n8 = For Each Loop(Array:Array of User Widget Object References=@n7) -> (Array Element:User Widget Object Reference)
   > n9 = Cast To PCPT_StatAvailable_ListItem(Object=@n8)
   n10 = "To String (Text)"(InText="GetText (Text Box)"!(self:Text Box Object Reference=Get txtBoxFilter() -> (txtBoxFilter:Text Box Object Reference)) -> (ReturnValue:Text))
   n11 = "integer > integer"(A=Len(S=@n10), B=0) {is=title K2Node_CallFunction}
   n12 = Contains(SearchIn=Get label(self=@n9), Substring=@n10)
   > SetVisibility(
    self=@n9
    InVisibility=Select(Option 0=Collapsed, Option 1=Visible, Index=Select(Option 0:bool=true, Option 1:bool=@n12, Index=@n11) -> (ReturnValue:bool))
   )
   "On Clicked (btnReturn)"!() -> (OutputDelegate, then, widget:User Widget Object Reference, event:Pointer Event Structure) {is=title K2Node_ComponentBoundEvent}
   > SetVisibility(InVisibility=Collapsed)
   "On Clicked (btnResetAll)"!() -> (OutputDelegate, then, widget:User Widget Object Reference, event:Pointer Event Structure) {is=title K2Node_ComponentBoundEvent}
   > toggleInclusionOnAll()
   "On Clicked (btnClearFilter)"!() -> (OutputDelegate, then, widget:User Widget Object Reference, event:Pointer Event Structure) {is=title K2Node_ComponentBoundEvent}
   > "SetText (Text Box)"!(execute, self:Text Box Object Reference=Get txtBoxFilter(), InText:Text) -> (then)
  ]}
 Should Stat be Included {nodes [
   n1 = Should Stat be Included!() -> (then, ItemToFind:"String (by ref)") {is=title K2Node_FunctionEntry}
   n2 = Contains Item!(TargetSet:Set of Strings=Get statsToIncludeByDefault() -> (statsToIncludeByDefault:Set of Strings), ItemToFind:"String (by ref)"=@n1) -> (ReturnValue:bool) {is=title K2Node_CallFunction}
   n3 = Length(TargetArray:Array of Strings=Get hiddenCategories() -> (hiddenCategories:Array of Strings))
   n4 = Contains Item(TargetArray:Array of Strings=Get hiddenCategories(), ItemToFind:"String (by ref)"=@n1) {is=title K2Node_CallArrayFunction}
   > Return Node!(execute, shouldInclude:bool=Select(Option 0=@n2, Option 1=NOT Boolean(A=@n4), Index="integer > integer"(A=@n3, B=0)))
   NOT Boolean(A=@n2)
   Get statsToIgnoreByDefault() -> (statsToIgnoreByDefault:Set of Strings)
  ]}
 handleStatInclusionChanged {nodes [
   n1 = handleStatInclusionChanged!() -> (then, widget:PCPT Stat Available List Item Object Reference, stat:String, included:bool=~) {is=title K2Node_FunctionEntry}
   > Call On Stat Inclusion Changed!(
    execute:Exec
    self:PCPT Show Stats List Widget Object Reference
    widget:PCPT Stat Available List Item Object Reference=@n1
    stat:String=@n1.stat
    included:bool=@n1.included
   ) -> (then:Exec) {is=title K2Node_CallDelegate}
  ]}
 onStatInclusionChanged {nodes [onStatInclusionChanged!() -> (then, widget:PCPT Stat Available List Item Object Reference, stat:String, included:bool=~) {is=title K2Node_FunctionEntry}]}
 requestHiddenCategories {nodes [
   requestHiddenCategories() {is=title K2Node_FunctionEntry}
   > n1 = Sequence()
   > n2 = getListItems(self=Get listStatAvailable())
   > n3 = For Each Loop(Array=@n2)
   > n4 = Cast To PCPT_StatAvailable_ListItem(Object=@n3)
   > n5 = Branch(Condition=getIsIncluded(self=@n4))
   Add(execute=@n5.else, TargetArray=Get l_hidden() -> (l_hidden:Array of Strings), NewItem=Get label(self=@n4)) {is=title K2Node_CallArrayFunction}
   Return Node!(execute=@n1.then_1, return:Array of Strings=Get l_hidden())
  ]}
 setAvailableStats {nodes [
   n1 = setAvailableStats!() -> (then, statList:Array of Strings) {is=title K2Node_FunctionEntry}
   > clearItems(self:TA Generic 01 List Object Reference=Get listStatAvailable())
   > n2 = For Each Loop(Array:Array of Strings=@n1) -> (Array Element:String)
   > n3 = Construct Object from Class(Class=/Game/Blueprints/Blutilities/ForTechArt/Data/TA_GenericListItemData.TA_GenericListItemData_C, self=Self-Reference()) -> (ReturnValue:TA Generic List Item Data Object Reference)
   > Set displayString!(execute, displayString:String=@n2, self:TA Generic List Item Data Object Reference=@n3) -> (then, Output_Get:String)
   > Set bools!(
    execute
    bools:Array of Booleans=Make Array("[0]":bool=Should Stat be Included!(ItemToFind:"String (by ref)"=@n2) -> (shouldInclude:bool) {is=title K2Node_CallFunction}) -> (Array:Array of Booleans)
    self:TA Generic List Item Data Object Reference=@n3
   ) -> (then, Output_Get:Array of Booleans)
   > n4 = addItem!(execute, self:TA Generic 01 List Object Reference=Get listStatAvailable(), dataObject:Object Reference=@n3) -> (then, newItem:User Widget Object Reference, newData:Object Reference)
   > n5 = Cast To PCPT_StatAvailable_ListItem(Object=@n4)
   > Bind Event to On Inclusion Changed(self:PCPT Stat Available List Item Object Reference=@n5, Delegate=Create Event()) {is=title K2Node_AddDelegate}
  ]}
 setHiddenCategories {nodes [
   n1 = setHiddenCategories!() -> (then, hiddenCategories:Array of Strings) {is=title K2Node_FunctionEntry}
   > Set hiddenCategories(hiddenCategories:Array of Strings=@n1) -> (Output_Get:Array of Strings)
   > n2 = getListItems(self=Get listStatAvailable())
   > n3 = For Each Loop(Array:Array of User Widget Object References=@n2) -> (Array Element:User Widget Object Reference)
   > n4 = Cast To PCPT_StatAvailable_ListItem(Object=@n3)
   > Set Included(self=@n4, isIncluded=Contains Item(TargetArray=Get hiddenCategories(), ItemToFind=Get label(self=@n4)), throwEvent=false)
  ]}
 toggleInclusionOnAll {nodes [
   toggleInclusionOnAll() {is=title K2Node_FunctionEntry}
   > n1 = getListItems(self=Get listStatAvailable())
   > n2 = Sequence()
   > Branch(Condition=Is Valid Index!(TargetArray:Array of User Widget Object References=@n1, IndexToTest:int) -> (ReturnValue:bool))
   > n3 = Cast To PCPT_StatAvailable_ListItem(Object="Get (a copy)"(Array:Array of User Widget Object References=@n1) -> (Output:User Widget Object Reference))
   > Set l_included(l_included:bool=NOT Boolean(A=getIsIncluded(self=@n3))) -> (Output_Get:bool)
   n4 = For Each Loop(Exec=@n2.then_1, Array=@n1)
   > n5 = Cast To PCPT_StatAvailable_ListItem(Object=@n4)
   > Set Included(self=@n5, isIncluded=Get l_included() -> (l_included:bool), throwEvent=true)
  ]}
]
variables [
 hiddenCategories {type="array<string>", default []}
 onStatInclusionChanged {default=()}
 stats {type="map<PCPT_Stat_Struct>", default {}}
 statsToIgnoreByDefault {
  type="array<string>"
  default [View/PosX, View/PosY, View/PosZ, View/ForwardX, View/ForwardY, View/ForwardZ, View/Speed, LightCount/ShadowOn, LightCount/ShadowOff]
 }
 statsToIncludeByDefault {type="array<string>", default [FrameTime, RHI/DrawCalls, DynamicResolutionPercentage, LightCount/All, GPU/ShadowDepths, GPU/Translucency]}
]
widgets {
 info {
  parentClass=/Game/Blueprints/Blutilities/ForTechArt/Widgets/TA_Base_EditorWidget.TA_Base_EditorWidget_C
  rootWidgetClass=&BackgroundBlur
  widgetCount=14
 }
 widgets [
  BackgroundBlur_0 {widgetClassPath=&BackgroundBlur, widget=BackgroundBlur_0, properties {blurStrength=5}}
  Border_0 {
   widgetClassPath=&Border
   parent=BackgroundBlur_0
   widget=Border_0
   properties {horizontalAlignment=HAlign_Center, brushColor=<rgba 0 0 0 0.688>, padding=<ltrb 10 50 10 50>}
   slot {class=BackgroundBlurSlot, properties {padding=<ltrb 0 0 0 0>}}
  }
  Border_1 {
   widgetClassPath=&Border
   parent=Border_0
   widget=Border_1
   properties {brushColor=<rgba 0.0365 0.0365 0.0365 1>, padding=<ltrb 0 0 0 0>}
   slot {class=BorderSlot, properties {horizontalAlignment=HAlign_Center, padding=<ltrb 10 50 10 50>}}
  }
  VerticalBox_0 {widgetClassPath=&VerticalBox, parent=Border_1, widget=VerticalBox_0, slot {class=BorderSlot, properties {padding=<ltrb 0 0 0 0>}}}
  Border_2 {
   widgetClassPath=&Border
   parent=VerticalBox_0
   widget=Border_2
   properties {brushColor=<rgba 0.1042 0.1042 0.1042 1>}
   slot {class=VerticalBoxSlot}
  }
  CommonTextBlock_129 {
   widgetClassPath=&CommonTextBlock
   parent=Border_2
   widget=CommonTextBlock_129
   properties {
    text=Select which stats to display
    font {
     bForceMonospaced=false
     bMaterialIsStencil=false
     fontMaterial=null
     fontObject=&NotoSans
     letterSpacing=0
     monospacedWidth=1
     size=16
     skewAmount=0
     typefaceFontName=NotoSans-Regular
     outlineSettings {
      bApplyOutlineToDropShadows=false
      bMiteredCorners=false
      bSeparateFillAlpha=false
      outlineMaterial=null
      outlineSize=0
      outlineColor=<rgba 0 0 0 1>
     }
    }
    shadowOffset=<0 0>
   }
   slot {class=BorderSlot}
  }
  VerticalBox_1 {
   widgetClassPath=&VerticalBox
   parent=VerticalBox_0
   widget=VerticalBox_1
   slot {class=VerticalBoxSlot, properties {padding=<ltrb 10 5 10 10>, size {sizeRule=Fill, value=1}}}
  }
  btnResetAll {
   widgetClassPath=&TA_Generic01_Button
   parent=VerticalBox_1
   widget=btnResetAll
   properties {label=Toggle All, buttonColor_Default=<rgba 1 0.4584 0.0999 1>}
   slot {class=VerticalBoxSlot}
  }
  HorizontalBox_0 {
   widgetClassPath=&HorizontalBox
   parent=VerticalBox_1
   widget=HorizontalBox_0
   slot {class=VerticalBoxSlot, properties {padding=<ltrb 0 5 0 0>}}
  }
  txtBoxFilter {
   widgetClassPath=&EditableTextBox
   bIsVariable=true
   parent=HorizontalBox_0
   widget=txtBoxFilter
   properties {
    hintText=Filter...
    widgetStyle {
     backgroundColor {colorUseRule=UseColor_Specified, specifiedColor=<rgba 1 1 1 1>}
     backgroundImageFocused {
      bIsDynamicallyLoaded=false
      drawAs=RoundedBox
      imageType=NoImage
      mirroring=NoMirror
      resourceName=None
      resourceObject=null
      tiling=NoTile
      imageSize=<0 0>
      margin=<ltrb 0 0 0 0>
      outlineSettings {
       bUseBrushTransparency=false
       roundingType=FixedRadius
       width=1
       color {colorUseRule=UseColor_ColorTable, specifiedColor=<rgba 0 0 0 0>}
       cornerRadii=<4 4 4 4>
      }
      tintColor {colorUseRule=UseColor_Specified, specifiedColor=<rgba 0.0048 0.0048 0.0048 1>}
      uVRegion {bIsValid=false, max=<0 0>, min=<0 0>}
     }
     backgroundImageHovered {
      bIsDynamicallyLoaded=false
      drawAs=RoundedBox
      imageType=NoImage
      mirroring=NoMirror
      resourceName=None
      resourceObject=null
      tiling=NoTile
      imageSize=<0 0>
      margin=<ltrb 0 0 0 0>
      outlineSettings {
       bUseBrushTransparency=false
       roundingType=FixedRadius
       width=1
       color {colorUseRule=UseColor_ColorTable, specifiedColor=<rgba 0 0 0 0>}
       cornerRadii=<4 4 4 4>
      }
      tintColor {colorUseRule=UseColor_Specified, specifiedColor=<rgba 0.0048 0.0048 0.0048 1>}
      uVRegion {bIsValid=false, max=<0 0>, min=<0 0>}
     }
     backgroundImageNormal {
      bIsDynamicallyLoaded=false
      drawAs=RoundedBox
      imageType=NoImage
      mirroring=NoMirror
      resourceName=None
      resourceObject=null
      tiling=NoTile
      imageSize=<0 0>
      margin=<ltrb 0 0 0 0>
      outlineSettings {
       bUseBrushTransparency=false
       roundingType=FixedRadius
       width=1
       color {colorUseRule=UseColor_ColorTable, specifiedColor=<rgba 0 0 0 0>}
       cornerRadii=<4 4 4 4>
      }
      tintColor {colorUseRule=UseColor_Specified, specifiedColor=<rgba 0.0048 0.0048 0.0048 1>}
      uVRegion {bIsValid=false, max=<0 0>, min=<0 0>}
     }
     backgroundImageReadOnly {
      bIsDynamicallyLoaded=false
      drawAs=RoundedBox
      imageType=NoImage
      mirroring=NoMirror
      resourceName=None
      resourceObject=null
      tiling=NoTile
      imageSize=<0 0>
      margin=<ltrb 0 0 0 0>
      outlineSettings {
       bUseBrushTransparency=false
       roundingType=FixedRadius
       width=1
       color {colorUseRule=UseColor_ColorTable, specifiedColor=<rgba 0 0 0 0>}
       cornerRadii=<4 4 4 4>
      }
      tintColor {colorUseRule=UseColor_Specified, specifiedColor=<rgba 0.0284 0.0284 0.0284 1>}
      uVRegion {bIsValid=false, max=<0 0>, min=<0 0>}
     }
     focusedForegroundColor {colorUseRule=UseColor_ColorTable, specifiedColor=<rgba 0 0 0 0>}
     foregroundColor {colorUseRule=UseColor_Specified, specifiedColor=<rgba 0.5271 0.5271 0.5271 1>}
     hScrollBarPadding=<ltrb 0 0 0 0>
     padding=<ltrb 8 3 8 4>
     readOnlyForegroundColor {colorUseRule=UseColor_Foreground, specifiedColor=<rgba 1 0 1 1>}
     scrollBarStyle {
      thickness=8
      draggedThumbImage {
       bIsDynamicallyLoaded=false
       drawAs=RoundedBox
       imageType=NoImage
       mirroring=NoMirror
       resourceName=None
       resourceObject=null
       tiling=NoTile
       imageSize=<0 0>
       margin=<ltrb 0 0 0 0>
       outlineSettings {
        bUseBrushTransparency=false
        roundingType=FixedRadius
        width=0
        color {colorUseRule=UseColor_Specified, specifiedColor=<rgba 0 0 0 0>}
        cornerRadii=<4 4 4 4>
       }
       tintColor {colorUseRule=UseColor_Specified, specifiedColor=<rgba 0.2159 0.2159 0.2159 1>}
       uVRegion {bIsValid=false, max=<0 0>, min=<0 0>}
      }
      horizontalBackgroundImage {
       bIsDynamicallyLoaded=false
       drawAs=NoDrawType
       imageType=NoImage
       mirroring=NoMirror
       resourceName=None
       resourceObject=null
       tiling=NoTile
       imageSize=<0 0>
       margin=<ltrb 0 0 0 0>
       outlineSettings {
        bUseBrushTransparency=false
        roundingType=HalfHeightRadius
        width=0
        color {colorUseRule=UseColor_Specified, specifiedColor=<rgba 0 0 0 0>}
        cornerRadii=<0 0 0 0>
       }
       tintColor {colorUseRule=UseColor_Specified, specifiedColor=<rgba 1 1 1 1>}
       uVRegion {bIsValid=false, max=<0 0>, min=<0 0>}
      }
      horizontalBottomSlotImage {
       bIsDynamicallyLoaded=false
       drawAs=NoDrawType
       imageType=NoImage
       mirroring=NoMirror
       resourceName=None
       resourceObject=null
       tiling=NoTile
       imageSize=<0 0>
       margin=<ltrb 0 0 0 0>
       outlineSettings {
        bUseBrushTransparency=false
        roundingType=HalfHeightRadius
        width=0
        color {colorUseRule=UseColor_Specified, specifiedColor=<rgba 0 0 0 0>}
        cornerRadii=<0 0 0 0>
       }
       tintColor {colorUseRule=UseColor_Specified, specifiedColor=<rgba 1 1 1 1>}
       uVRegion {bIsValid=false, max=<0 0>, min=<0 0>}
      }
      horizontalTopSlotImage {
       bIsDynamicallyLoaded=false
       drawAs=NoDrawType
       imageType=NoImage
       mirroring=NoMirror
       resourceName=None
       resourceObject=null
       tiling=NoTile
       imageSize=<0 0>
       margin=<ltrb 0 0 0 0>
       outlineSettings {
        bUseBrushTransparency=false
        roundingType=HalfHeightRadius
        width=0
        color {colorUseRule=UseColor_Specified, specifiedColor=<rgba 0 0 0 0>}
        cornerRadii=<0 0 0 0>
       }
       tintColor {colorUseRule=UseColor_Specified, specifiedColor=<rgba 1 1 1 1>}
       uVRegion {bIsValid=false, max=<0 0>, min=<0 0>}
      }
      hoveredThumbImage {
       bIsDynamicallyLoaded=false
       drawAs=RoundedBox
       imageType=NoImage
       mirroring=NoMirror
       resourceName=None
       resourceObject=null
       tiling=NoTile
       imageSize=<0 0>
       margin=<ltrb 0 0 0 0>
       outlineSettings {
        bUseBrushTransparency=false
        roundingType=FixedRadius
        width=0
        color {colorUseRule=UseColor_Specified, specifiedColor=<rgba 0 0 0 0>}
        cornerRadii=<4 4 4 4>
       }
       tintColor {colorUseRule=UseColor_Specified, specifiedColor=<rgba 0.2159 0.2159 0.2159 1>}
       uVRegion {bIsValid=false, max=<0 0>, min=<0 0>}
      }
      normalThumbImage {
       bIsDynamicallyLoaded=false
       drawAs=RoundedBox
       imageType=NoImage
       mirroring=NoMirror
       resourceName=None
       resourceObject=null
       tiling=NoTile
       imageSize=<0 0>
       margin=<ltrb 0 0 0 0>
       outlineSettings {
        bUseBrushTransparency=false
        roundingType=FixedRadius
        width=0
        color {colorUseRule=UseColor_Specified, specifiedColor=<rgba 0 0 0 0>}
        cornerRadii=<4 4 4 4>
       }
       tintColor {colorUseRule=UseColor_Specified, specifiedColor=<rgba 0.0953 0.0953 0.0953 1>}
       uVRegion {bIsValid=false, max=<0 0>, min=<0 0>}
      }
      verticalBackgroundImage {
       bIsDynamicallyLoaded=false
       drawAs=NoDrawType
       imageType=NoImage
       mirroring=NoMirror
       resourceName=None
       resourceObject=null
       tiling=NoTile
       imageSize=<0 0>
       margin=<ltrb 0 0 0 0>
       outlineSettings {
        bUseBrushTransparency=false
        roundingType=HalfHeightRadius
        width=0
        color {colorUseRule=UseColor_Specified, specifiedColor=<rgba 0 0 0 0>}
        cornerRadii=<0 0 0 0>
       }
       tintColor {colorUseRule=UseColor_Specified, specifiedColor=<rgba 1 1 1 1>}
       uVRegion {bIsValid=false, max=<0 0>, min=<0 0>}
      }
      verticalBottomSlotImage {
       bIsDynamicallyLoaded=false
       drawAs=NoDrawType
       imageType=NoImage
       mirroring=NoMirror
       resourceName=None
       resourceObject=null
       tiling=NoTile
       imageSize=<0 0>
       margin=<ltrb 0 0 0 0>
       outlineSettings {
        bUseBrushTransparency=false
        roundingType=HalfHeightRadius
        width=0
        color {colorUseRule=UseColor_Specified, specifiedColor=<rgba 0 0 0 0>}
        cornerRadii=<0 0 0 0>
       }
       tintColor {colorUseRule=UseColor_Specified, specifiedColor=<rgba 1 1 1 1>}
       uVRegion {bIsValid=false, max=<0 0>, min=<0 0>}
      }
      verticalTopSlotImage {
       bIsDynamicallyLoaded=false
       drawAs=NoDrawType
       imageType=NoImage
       mirroring=NoMirror
       resourceName=None
       resourceObject=null
       tiling=NoTile
       imageSize=<0 0>
       margin=<ltrb 0 0 0 0>
       outlineSettings {
        bUseBrushTransparency=false
        roundingType=HalfHeightRadius
        width=0
        color {colorUseRule=UseColor_Specified, specifiedColor=<rgba 0 0 0 0>}
        cornerRadii=<0 0 0 0>
       }
       tintColor {colorUseRule=UseColor_Specified, specifiedColor=<rgba 1 1 1 1>}
       uVRegion {bIsValid=false, max=<0 0>, min=<0 0>}
      }
     }
     textStyle {
      overflowPolicy=Clip
      transformPolicy=None
      colorAndOpacity {colorUseRule=UseColor_Foreground, specifiedColor=<rgba 1 0 1 1>}
      font {
       bForceMonospaced=false
       bMaterialIsStencil=false
       fontMaterial=null
       fontObject=null
       letterSpacing=0
       monospacedWidth=1
       size=10
       skewAmount=0
       typefaceFontName=Regular
       outlineSettings {
        bApplyOutlineToDropShadows=false
        bMiteredCorners=false
        bSeparateFillAlpha=false
        outlineMaterial=null
        outlineSize=0
        outlineColor=<rgba 0 0 0 1>
       }
      }
      highlightColor {colorUseRule=UseColor_ColorTable, specifiedColor=<rgba 0 0 0 0>}
      highlightShape {
       bIsDynamicallyLoaded=false
       drawAs=Image
       imageType=NoImage
       mirroring=NoMirror
       resourceName=None
       resourceObject=null
       tiling=NoTile
       imageSize=<0 0>
       margin=<ltrb 0 0 0 0>
       outlineSettings {
        bUseBrushTransparency=false
        roundingType=HalfHeightRadius
        width=0
        color {colorUseRule=UseColor_Specified, specifiedColor=<rgba 0 0 0 0>}
        cornerRadii=<0 0 0 0>
       }
       tintColor {colorUseRule=UseColor_ColorTable, specifiedColor=<rgba 0 0 0 0>}
       uVRegion {bIsValid=false, max=<0 0>, min=<0 0>}
      }
      selectedBackgroundColor {colorUseRule=UseColor_ColorTable, specifiedColor=<rgba 0 0 0 0>}
      shadowColorAndOpacity=<rgba 0 0 0 1>
      shadowOffset=<0 0>
      strikeBrush {
       bIsDynamicallyLoaded=false
       drawAs=Image
       imageType=NoImage
       mirroring=NoMirror
       resourceName=None
       resourceObject=null
       tiling=NoTile
       imageSize=<32 32>
       margin=<ltrb 0 0 0 0>
       outlineSettings {
        bUseBrushTransparency=false
        roundingType=HalfHeightRadius
        width=0
        color {colorUseRule=UseColor_Specified, specifiedColor=<rgba 0 0 0 0>}
        cornerRadii=<0 0 0 0>
       }
       tintColor {colorUseRule=UseColor_Specified, specifiedColor=<rgba 1 1 1 1>}
       uVRegion {bIsValid=false, max=<0 0>, min=<0 0>}
      }
      underlineBrush {
       bIsDynamicallyLoaded=false
       drawAs=Image
       imageType=NoImage
       mirroring=NoMirror
       resourceName=None
       resourceObject=null
       tiling=NoTile
       imageSize=<32 32>
       margin=<ltrb 0 0 0 0>
       outlineSettings {
        bUseBrushTransparency=false
        roundingType=HalfHeightRadius
        width=0
        color {colorUseRule=UseColor_Specified, specifiedColor=<rgba 0 0 0 0>}
        cornerRadii=<0 0 0 0>
       }
       tintColor {colorUseRule=UseColor_Specified, specifiedColor=<rgba 1 1 1 1>}
       uVRegion {bIsValid=false, max=<0 0>, min=<0 0>}
      }
     }
     vScrollBarPadding=<ltrb 0 0 0 0>
    }
   }
   slot {class=HorizontalBoxSlot, properties {size {sizeRule=Fill, value=1}}}
  }
  btnClearFilter {
   widgetClassPath=&TA_Generic01_Button
   parent=HorizontalBox_0
   widget=btnClearFilter
   properties {
    buttonColor_Default=<rgba 1 0.0285 0.0982 1>
    iconBrush {
     bIsDynamicallyLoaded=false
     drawAs=Image
     imageType=NoImage
     mirroring=NoMirror
     resourceName=None
     resourceObject=/Game/UI/Foundation/Textures/Buttons/T-Button-Clear-64
     tiling=NoTile
     imageSize=<16 16>
     margin=<ltrb 0 0 0 0>
     outlineSettings {
      bUseBrushTransparency=false
      roundingType=HalfHeightRadius
      width=0
      color {colorUseRule=UseColor_Specified, specifiedColor=<rgba 0 0 0 0>}
      cornerRadii=<0 0 0 1>
     }
     tintColor {colorUseRule=UseColor_Specified, specifiedColor=<rgba 1 1 1 1>}
     uVRegion {bIsValid=false, max=<0 0>, min=<0 0>}
    }
   }
   slot {class=HorizontalBoxSlot}
  }
  Border_3 {
   widgetClassPath=&Border
   parent=VerticalBox_1
   widget=Border_3
   properties {brushColor=<rgba 0.0104 0.0104 0.0104 0.324>}
   slot {class=VerticalBoxSlot, properties {padding=<ltrb 0 5 0 0>, size {sizeRule=Fill, value=1}}}
  }
  listStatAvailable {
   widgetClassPath=/Game/Blueprints/Blutilities/ForTechArt/Widgets/List/TA_Generic01_List
   bIsVariable=true
   parent=Border_3
   widget=listStatAvailable
   properties {
    allowMultiSelect=false
    allowNoSelection=true
    enableSelection=false
    listItemClass=/Game/NeverCook/EditorUtilities/PerfTools/PerfCsvPlaybackTool/SubWidgets/PCPT_StatAvailable_ListItem.PCPT_StatAvailable_ListItem_C
   }
   slot {class=BorderSlot}
  }
  btnReturn {
   widgetClassPath=&TA_Generic01_Button
   parent=VerticalBox_1
   widget=btnReturn
   properties {label=Return, buttonColor_Default=<rgba 0.1142 0.6771 0.1881 1>}
   slot {class=VerticalBoxSlot, properties {padding=<ltrb 0 5 0 0>}}
  }
 ]
}
