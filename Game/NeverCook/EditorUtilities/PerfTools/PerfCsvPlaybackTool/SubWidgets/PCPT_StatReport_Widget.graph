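# Stat tile widget for the editor-side Perf CSV Playback tool, showing a stat name above a numeric value in a bordered selectable button.
# setListItemData reads the key and float series from TA_GenericListItemData and shows the first value.
# setIndex displays the sample at a given index, while setDataAndVisibility replaces the series, writes it back to the item and shows or collapses the tile.
# setFocused tints the border red, and setColor recolors the value text.
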WidgetBlueprint /Game/NeverCook/EditorUtilities/PerfTools/PerfCsvPlaybackTool/SubWidgets/PCPT_StatReport_Widget
graphs [
 EventGraph {nodes [
   n1 = Event PreConstruct()
   > Parent: PreConstruct(IsDesignTime=@n1.IsDesignTime)
   > setKey!(execute, self, key:String=Get DefaultKey() -> (DefaultKey:String)) -> (then)
   n2 = Event setListItemData!() -> (OutputDelegate, then, listItemData:Object Reference) {is=title K2Node_Event}
   > Parent: setListItemData!(execute, listItemData:Object Reference=@n2.listItemData) -> (then) {is=title K2Node_CallParentFunction}
   > n3 = Cast To TA_GenericListItemData!(execute, Object=@n2.listItemData) -> (then, CastFailed, AsTA Generic List Item Data:TA Generic List Item Data Object Reference) {is=title K2Node_DynamicCast}
   > n4 = Set DefaultKey(DefaultKey:String=Get displayString!(self:TA Generic List Item Data Object Reference=@n3) -> (displayString:String)) -> (Output_Get:String)
   > setKey(key=@n4)
   n5 = Get floats!(self:TA Generic List Item Data Object Reference=@n3) -> (floats:"Array of Float (double-precision)s")
   setValue!(
    execute
    self
    valueText:Text="To Text (Float)"(
     Value="Get (a copy)"(Array:"Array of Float (double-precision)s"=@n5) -> (Output:double)
     MinimumIntegralDigits=1
     MaximumIntegralDigits=324
     MinimumFractionalDigits=0
     MaximumFractionalDigits=2
    )
   ) -> (then)
   Set data(data:"Array of Float (double-precision)s"=@n5) -> (Output_Get:"Array of Float (double-precision)s")
  ]}
 SetVis {nodes [SetVis() {is=title K2Node_FunctionEntry}]}
 getBackgroundWidget {nodes [
   getBackgroundWidget() {is=title K2Node_FunctionEntry}
   > Return Node!(execute, backgroundWidget:Widget Object Reference=Get Border_0() -> (Border_0:Border Object Reference))
  ]}
 getKey {nodes [
   getKey() {is=title K2Node_FunctionEntry}
   n1 = Get txtKey() -> (txtKey:Common Text Object Reference)
   > Return Node!(execute, key:String="To String (Text)"(InText="GetText (Text)"!(self:Text Object Reference=@n1) -> (ReturnValue:Text)))
  ]}
 setColor {nodes [
   n1 = setColor!() -> (then, color:Linear Color Structure) {is=title K2Node_FunctionEntry}
   > SetColorAndOpacity(self=Get txtVal() -> (txtVal:Common Text Object Reference), InColorAndOpacity=Make Slate Color(SpecifiedColor=@n1))
  ]}
 setDataAndVisibility {nodes [
   n1 = setDataAndVisibility!() -> (then, vis:bool=~, data:"Array of Float (double-precision)s", index:int) {is=title K2Node_FunctionEntry}
   > n2 = Branch(Condition=@n1)
   > Set data(data=@n1.data)
   > n3 = Cast To TA_GenericListItemData(Object=Get currentListItemData() -> (currentListItemData:Object Reference))
   > Set floats!(
    execute
    floats:"Array of Float (double-precision)s"=Get data() -> (data:"Array of Float (double-precision)s")
    self:TA Generic List Item Data Object Reference=@n3
   ) -> (then, Output_Get:"Array of Float (double-precision)s")
   > SetVisibility(InVisibility=Visible)
   > setIndex!(execute, self, index:int=@n1.index) -> (then)
   SetVisibility(execute=@n2.else, InVisibility=Collapsed)
  ]}
 setFocused {nodes [
   n1 = setFocused!() -> (then, focused:bool=~) {is=title K2Node_FunctionEntry}
   > setSelected!(execute, self:TA Base Editor Button Widget Object Reference, selected:bool=@n1, throwEvent:bool=false) -> (then)
   SetBrushColor(self=Get Border_0(), InBrushColor=SelectColor(A=(R=1,G=0.0999,B=0.0497,A=1), B=(R=0.0194,G=0.0194,B=0.0194,A=1), bPickA=@n1))
  ]}
 setIndex {nodes [
   n1 = setIndex!() -> (then, index:int) {is=title K2Node_FunctionEntry}
   n2 = Get data()
   > Branch(Condition=Is Valid Index!(TargetArray:"Array of Float (double-precision)s"=@n2, IndexToTest:int=@n1) -> (ReturnValue:bool))
   > setValue(valueText="To Text (Float)"(
     Value="Get (a copy)"(Array=@n2, Dimension 1=@n1)
     MinimumIntegralDigits=1
     MaximumIntegralDigits=324
     MinimumFractionalDigits=0
     MaximumFractionalDigits=2
    ))
  ]}
 setKey {nodes [
   n1 = setKey!() -> (then, key:String) {is=title K2Node_FunctionEntry}
   > "SetText (Text)"(self=Get txtKey(), InText="To Text (String)"(InString=@n1))
  ]}
 setValue {nodes [
   n1 = setValue!() -> (then, valueText:Text) {is=title K2Node_FunctionEntry}
   > "SetText (Text)"(self=Get txtVal(), InText=@n1)
  ]}
]
variables [
 data {type="array<number>", default []}
 DefaultKey {type=string, default=Key}
]
widgets {
 info {
  inheritedWidgetCount=0
  parentClass=/Game/Blueprints/Blutilities/ForTechArt/Widgets/Buttons/TA_BaseEditorButtonWidget.TA_BaseEditorButtonWidget_C
  rootWidgetClass=&CanvasPanel
  widgetCount=7
 }
 widgets [
  CanvasPanel_0 {widgetClassPath=&CanvasPanel, widget=CanvasPanel_0}
  SizeBox_0 {
   widgetClassPath=&SizeBox
   parent=CanvasPanel_0
   widget=SizeBox_0
   properties {bOverride_MinDesiredWidth=true, minDesiredWidth=80}
   slot {
    class=CanvasPanelSlot
    properties {bAutoSize=true, layoutData {alignment=<0 0>, anchors {maximum=<0 0>, minimum=<0 0>}, offsets=<ltrb 0 0 75.976 35.8559>}}
   }
  }
  Border_0 {
   widgetClassPath=&Border
   bIsVariable=true
   parent=SizeBox_0
   widget=Border_0
   properties {brushColor=<rgba 0.0197 0.0222 0.026 1>, padding=<ltrb 3 3 3 3>}
   slot {class=SizeBoxSlot, properties {padding=<ltrb 2.5 0 2.5 0>}}
  }
  Border_1 {
   widgetClassPath=&Border
   parent=Border_0
   widget=Border_1
   properties {brushColor=<rgba 0.0255 0.0354 0.05 1>, padding=<ltrb 1 1 1 1>}
   slot {class=BorderSlot, properties {padding=<ltrb 3 3 3 3>}}
  }
  VerticalBox_0 {widgetClassPath=&VerticalBox, parent=Border_1, widget=VerticalBox_0, slot {class=BorderSlot, properties {padding=<ltrb 1 1 1 1>}}}
  txtKey {
   widgetClassPath=&CommonTextBlock
   bIsVariable=true
   parent=VerticalBox_0
   widget=txtKey
   properties {
    text=Text Block
    colorAndOpacity {colorUseRule=UseColor_Specified, specifiedColor=<rgba 0.9323 0.9323 0.9323 1>}
    font {
     bForceMonospaced=false
     bMaterialIsStencil=false
     fontMaterial=null
     fontObject=&NotoSans
     letterSpacing=0
     monospacedWidth=1
     size=10
     skewAmount=0
     typefaceFontName=NotoSans-Regular
     outlineSettings {
      bApplyOutlineToDropShadows=false
      bMiteredCorners=false
      bSeparateFillAlpha=false
      outlineMaterial=null
      outlineSize=0
      outlineColor=<rgba 0 0 0 1>
     }
    }
    shadowColorAndOpacity=<rgba 0 0 0 1>
    shadowOffset=<0.25 0.25>
   }
   slot {class=VerticalBoxSlot, properties {horizontalAlignment=HAlign_Center}}
  }
  txtVal {
   widgetClassPath=&CommonTextBlock
   bIsVariable=true
   parent=VerticalBox_0
   widget=txtVal
   properties {
    text=Value
    font {
     bForceMonospaced=false
     bMaterialIsStencil=false
     fontMaterial=null
     fontObject=&NotoSans
     letterSpacing=0
     monospacedWidth=1
     size=10
     skewAmount=0
     typefaceFontName=NotoSans-Bold
     outlineSettings {
      bApplyOutlineToDropShadows=false
      bMiteredCorners=false
      bSeparateFillAlpha=false
      outlineMaterial=null
      outlineSize=0
      outlineColor=<rgba 0 0 0 1>
     }
    }
    shadowColorAndOpacity=<rgba 0 0 0 1>
    shadowOffset=<0.25 0.25>
   }
   slot {class=VerticalBoxSlot, properties {horizontalAlignment=HAlign_Center}}
  }
 ]
}
